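# Physics asset for the EmberRae medium female player body, covering the lower body and two dynamic skirt chains rather than a full ragdoll.
# Capsules cover spine_01, pelvis, legs, thigh twist bones, feet and thigh armor bases, all left unconstrained.
# Constraints link only the three-segment front skirt and four-segment bottom skirt, each chained from its base bone.
# Tight swing and twist limits keep both chains fairly stiff, the front allowing less motion.
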PhysicsAsset /BRCosmetics/Characters/Player/Female/Medium/Bodies/F_MED_EmberRae/Meshes/F_MED_EmberRae_Physics
bodies {
 calf_l [{shapeType=Capsule, length=15.0039, radius=11.0059, shapeName=calf_l_capsule, center=<-20.6923 -4.1492 0>, rotation=<rot 90 180 177.9868>}]
 calf_r [{
   shapeType=Capsule
   length=16.7213
   radius=11.1297
   shapeName=calf_r_capsule
   center=<19.8979 4.1852 0.0556>
   rotation=<rot -88.1628 4.9964 0.0014>
  }]
 dyn_bottom_skirt_01 [{
   shapeType=Capsule
   length=15.3415
   radius=2.2029
   shapeName=dyn_bottom_skirt_01_capsule
   center=<4.3262 0 0.4739>
   rotation=<rot -0.0007 0 -90>
  }]
 dyn_bottom_skirt_02 [{
   shapeType=Capsule
   length=15.6674
   radius=2.2153
   shapeName=dyn_bottom_skirt_02_capsule
   center=<3.8794 0 0.1073>
   rotation=<rot -0.0007 0 -90>
  }]
 dyn_bottom_skirt_03 [{shapeType=Capsule, length=12.076, radius=2.308, shapeName=dyn_bottom_skirt_03_capsule, center=<3.8945 0 0>, rotation=<rot -0.0001 0 90>}]
 dyn_bottom_skirt_04 [{
   shapeType=Capsule
   length=10.8292
   radius=2.0039
   shapeName=dyn_bottom_skirt_04_capsule
   center=<2.9824 0 -0.0618>
   rotation=<rot -0.0001 0 90>
  }]
 dyn_bottom_skirt_base [{
   shapeType=Capsule
   length=26.2216
   radius=5.973
   shapeName=dyn_bottom_skirt_base_capsule
   center=<13.5567 0 -8.9339>
   rotation=<rot 0 90 -96.4638>
  }]
 dyn_front_skirt_01 [{shapeType=Capsule, length=16.4442, radius=1.7582, shapeName=dyn_front_skirt_01_capsule, center=<2.6457 0.1286 0>}]
 dyn_front_skirt_02 [{shapeType=Capsule, length=13.8858, radius=1.85, shapeName=dyn_front_skirt_02_capsule, center=<2.4907 -0.6526 0>}]
 dyn_front_skirt_03 [{shapeType=Capsule, length=13.4495, radius=1.83, shapeName=dyn_front_skirt_03_capsule, center=<2.6542 -1.0417 0>}]
 dyn_front_skirt_base [{shapeType=Capsule, length=11.7121, radius=2.1469, shapeName=dyn_front_skirt_base_capsule, center=<-3.4774 -1.6385 0>}]
 foot_l [{shapeType=Capsule, length=14.7004, radius=6.1208, shapeName=foot_l_capsule, center=<1.9046 -1.6134 0>, rotation=<rot 90 -3.6023 -17.1642>}]
 foot_r [{shapeType=Capsule, length=15.0117, radius=5.9699, shapeName=foot_r_capsule, center=<-2.5144 0.9973 -0.3928>, rotation=<rot 90 180 170.4836>}]
 pelvis [{shapeType=Capsule, length=14.6021, radius=10.6479, shapeName=pelvis_capsule, center=<-8.1888 -2.8076 0>}]
 spine_01 [{
   shapeType=Capsule
   length=5.6171
   radius=11.9871
   shapeName=spine_01_capsule
   center=<-3.0062 0.0007 0>
   rotation=<rot 90 -179.9956 -179.9944>
  }]
 thigh_armor_l_base [{
   shapeType=Capsule
   length=35.3795
   radius=6.5539
   shapeName=thigh_armor_l_base_capsule
   center=<11.7 0.0004 -7.3538>
   rotation=<rot 85.6282 -0.0019 -0.0041>
  }]
 thigh_armor_r_base [{
   shapeType=Capsule
   length=35.5337
   radius=6.5073
   shapeName=thigh_armor_r_base_capsule
   center=<12.5938 0.0001 7.1679>
   rotation=<rot -86.0748 -0.0109 0.0137>
  }]
 thigh_l [{
   shapeType=Capsule
   length=29
   radius=9.5
   shapeName=thigh_l_capsule
   center=<-17.944 -1.5373 0.7454>
   rotation=<rot 87.4094 179.9596 -0.0026>
  }]
 thigh_r [{
   shapeType=Capsule
   length=28.7199
   radius=9.7801
   shapeName=thigh_r_capsule
   center=<18.1813 2.6379 -0.7839>
   rotation=<rot 86.5503 179.4638 179.4517>
  }]
 thigh_twist_01_l [{
   shapeType=Capsule
   length=28.4762
   radius=6.4063
   shapeName=thigh_twist_01_l_capsule
   center=<4.7891 -7.7587 3.547>
   rotation=<rot -89.2646 -179.9943 -177.9925>
  }]
 thigh_twist_01_r [{
   shapeType=Capsule
   length=27.1563
   radius=6.3996
   shapeName=thigh_twist_01_r_capsule
   center=<-4.8667 7.9148 -3.9982>
   rotation=<rot -86.6445 -179.011 1.1248>
  }]
}
constraints [
 {bone1Name=dyn_front_skirt_02, bone2Name=dyn_front_skirt_01, swing1LimitDegrees=4, swing2LimitDegrees=15, twistLimitDegrees=2}
 {bone1Name=dyn_front_skirt_03, bone2Name=dyn_front_skirt_02, swing1LimitDegrees=4, swing2LimitDegrees=10, twistLimitDegrees=2}
 {bone1Name=dyn_bottom_skirt_02, bone2Name=dyn_bottom_skirt_01, swing1LimitDegrees=5, swing2LimitDegrees=20, twistLimitDegrees=5}
 {bone1Name=dyn_bottom_skirt_03, bone2Name=dyn_bottom_skirt_02, swing1LimitDegrees=5, swing2LimitDegrees=20, twistLimitDegrees=5}
 {bone1Name=dyn_bottom_skirt_04, bone2Name=dyn_bottom_skirt_03, swing1LimitDegrees=5, swing2LimitDegrees=20, twistLimitDegrees=5}
 {bone1Name=dyn_front_skirt_01, bone2Name=dyn_front_skirt_base, swing1LimitDegrees=4, swing2LimitDegrees=15, twistLimitDegrees=0}
 {bone1Name=dyn_bottom_skirt_01, bone2Name=dyn_bottom_skirt_base, swing1LimitDegrees=5, swing2LimitDegrees=20, twistLimitDegrees=5}
]
