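# Animation Blueprint for the Ember Rae character body that drives skirt, thigh armor and left shoulder armor secondary motion.
# The copied parent pose passes through a Control Rig and corrective bone offsets weighted by limb rotations measured each update.
# Movement-state blends select RigidBody simulation variants, then constraints and offsets adjust for capes, backpacks and skydive diving.
# The EventGraph also feeds foot motion and skydive wind to materials for WPO cloth effects.
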AnimBlueprint /BRCosmetics/Characters/Player/Female/Medium/Bodies/F_MED_EmberRae/Meshes/F_MED_EmberRae_AnimBP
target_skeleton=&/F_MED_EmberRae_Skeleton
graphs [
 AnimGraph {nodes [
   "Transform (Modify) Bone - Bone: dyn_front_skirt_base"(Alpha=1, Translation="-12.568,8.690,0.226", Rotation="0,0,0", Scale="1,1,1") {
    rotationMode=BMM_Additive
    rotationSpace=BCS_BoneSpace
    scaleMode=BMM_Additive
    scaleSpace=BCS_BoneSpace
    translationMode=BMM_Additive
    translationSpace=BCS_BoneSpace
    alphaScaleBiasClamp {bInterpResult=true, interpSpeedDecreasing=1}
    boneToModify {boneName=dyn_front_skirt_base}
    translation=<-12.568 8.69 0.226>
   }
   n1 = "Transform (Modify) Bone - Bone: dyn_front_skirt_base"(ComponentPose=Local To Component(LocalPose=Use cached pose 'Pose02'()), Alpha=1, Translation="0,2,0", Rotation="-8,-10,15", Scale="1,1,1") {
    rotationMode=BMM_Additive
    rotationSpace=BCS_BoneSpace
    translationMode=BMM_Additive
    translationSpace=BCS_BoneSpace
    alphaBoolBlend {blendOption=Linear}
    boneToModify {boneName=dyn_front_skirt_base}
    rotation=<rot -8 -10 15>
    translation {y=2}
   }
   n2 = RigidBody(ComponentPose=@n1) {
    overridePhysicsAsset=&/F_MED_EmberRae_Physics
    simulationSpace=BaseBoneSpace
    alphaBoolBlend {blendOption=Linear}
   }
   n3 = "Transform (Modify) Bone - Bone: dyn_front_skirt_base"(ComponentPose=Local To Component(LocalPose=Use cached pose 'Pose02'()), Alpha=1, Translation="0,2,0", Rotation="-8,-10,15", Scale="1,1,1") {
    rotationMode=BMM_Additive
    rotationSpace=BCS_BoneSpace
    translationMode=BMM_Additive
    translationSpace=BCS_BoneSpace
    alphaBoolBlend {blendOption=Linear}
    boneToModify {boneName=dyn_front_skirt_base}
    rotation=<rot -8 -10 15>
    translation {y=2}
   }
   n4 = RigidBody(ComponentPose=@n3) {
    overridePhysicsAsset=&F_MED_EmberRae_Physics
    simulationSpace=BaseBoneSpace
    alphaBoolBlend {blendOption=Linear}
   }
   n5 = "Constraint - dyn_bottom_skirt_01 (P)"(ComponentPose=@n4, Alpha=0.3, ConstraintWeights_0=1) {
    alphaBoolBlend {blendOption=Linear}
    boneToModify {boneName=dyn_bottom_skirt_01}
    constraintSetup [{targetBone {boneName=dyn_bottom_skirt_base}}]
    constraintWeights [1]
   }
   n6 = "Constraint - dyn_bottom_skirt_02 (P)"(ComponentPose=@n5, Alpha=0.3, ConstraintWeights_0=1) {
    alphaBoolBlend {blendOption=Linear}
    boneToModify {boneName=dyn_bottom_skirt_02}
    constraintSetup [{targetBone {boneName=dyn_bottom_skirt_01}}]
    constraintWeights [1]
   }
   n7 = "Constraint - dyn_bottom_skirt_03 (P)"(ComponentPose=@n6, Alpha=0.5, ConstraintWeights_0=1) {
    alphaBoolBlend {blendOption=Linear}
    boneToModify {boneName=dyn_bottom_skirt_03}
    constraintSetup [{targetBone {boneName=dyn_bottom_skirt_02}}]
    constraintWeights [1]
   }
   n8 = "Constraint - dyn_bottom_skirt_04 (P)"(ComponentPose=@n7, Alpha=0.5, ConstraintWeights_0=1) {
    alphaBoolBlend {blendOption=Linear}
    boneToModify {boneName=dyn_bottom_skirt_04}
    constraintSetup [{targetBone {boneName=dyn_bottom_skirt_03}}]
    constraintWeights [1]
   }
   n9 = Blend Poses by bool(BlendPose_0=Component To Local(ComponentPose=@n2), BlendPose_1=Component To Local(ComponentPose=@n8), bActiveValue=Get bIsSkyDiving()) {blendType=Linear}
   n10 = "Transform (Modify) Bone - Bone: dyn_front_skirt_base"(
    ComponentPose=Local To Component(LocalPose=Use cached pose 'Pose02'())
    Alpha=Get Alpha_thighL_Z_minus() -> (Alpha_thighL_Z_minus:double)
    Translation="0,7,0"
    Rotation="10,40,40"
    Scale="1,1,1"
   ) {
    rotationMode=BMM_Additive
    rotationSpace=BCS_BoneSpace
    translationMode=BMM_Additive
    translationSpace=BCS_BoneSpace
    alphaBoolBlend {blendOption=Linear}
    boneToModify {boneName=dyn_front_skirt_base}
    rotation=<rot 10 40 40>
    translation {y=7}
   }
   n11 = "Transform (Modify) Bone - Bone: dyn_front_skirt_base"(
    ComponentPose=@n10
    Alpha=Get Alpha_thighR_Z_minus() -> (Alpha_thighR_Z_minus:double)
    Translation="0,7,0"
    Rotation="-10,40,-40"
    Scale="1,1,1"
   ) {
    rotationMode=BMM_Additive
    rotationSpace=BCS_BoneSpace
    translationMode=BMM_Additive
    translationSpace=BCS_BoneSpace
    alphaBoolBlend {blendOption=Linear}
    boneToModify {boneName=dyn_front_skirt_base}
    rotation=<rot -10 40 -40>
    translation {y=7}
   }
   n12 = RigidBody(ComponentPose=@n11) {
    overridePhysicsAsset=&F_MED_EmberRae_Physics
    simulationSpace=BaseBoneSpace
    alphaBoolBlend {blendOption=Linear}
   }
   n13 = Component To Local(ComponentPose=@n12)
   n14 = Get IsStreightMoving() -> (IsStreightMoving:bool)
   n15 = RigidBody(ComponentPose=Local To Component(LocalPose=Use cached pose 'Pose02'())) {
    bOverrideWorldGravity=true
    overridePhysicsAsset=&F_MED_EmberRae_Physics
    simulationSpace=BaseBoneSpace
    alphaBoolBlend {blendOption=Linear}
    overrideWorldGravity {z=2000}
   }
   n16 = "Transform (Modify) Bone - Bone: dyn_bottom_skirt_base"!(ComponentPose=@n15, bAlphaBoolEnabled:bool=True, Translation="1,0,1", Rotation="30,0,0", Scale="1,1,1") -> (Pose) {
    alphaInputType=Bool
    rotationMode=BMM_Additive
    rotationSpace=BCS_BoneSpace
    translationMode=BMM_Additive
    translationSpace=BCS_BoneSpace
    alphaBoolBlend {blendOption=Linear}
    boneToModify {boneName=dyn_bottom_skirt_base}
    rotation {pitch=30}
    translation {x=1, z=1}
   }
   n17 = "Transform (Modify) Bone - Bone: dyn_front_skirt_base"!(ComponentPose=@n16, bAlphaBoolEnabled:bool=True, Translation="-14,9,4", Rotation="0,50,-40", Scale="1,1,1") -> (Pose) {
    alphaInputType=Bool
    rotationMode=BMM_Additive
    rotationSpace=BCS_BoneSpace
    translationMode=BMM_Additive
    translationSpace=BCS_BoneSpace
    alphaBoolBlend {blendOption=Linear}
    boneToModify {boneName=dyn_front_skirt_base}
    rotation {roll=-40, yaw=50}
    translation=<-14 9 4>
   }
   n18 = "Transform (Modify) Bone - Bone: thigh_armor_l_base"!(ComponentPose=@n17, bAlphaBoolEnabled:bool=True, Translation="5,-1,0", Rotation="-5,0,0", Scale="1,1,1") -> (Pose) {
    alphaInputType=Bool
    rotationMode=BMM_Additive
    rotationSpace=BCS_BoneSpace
    scaleSpace=BCS_BoneSpace
    translationMode=BMM_Additive
    translationSpace=BCS_BoneSpace
    alphaBoolBlend {blendOption=Linear}
    boneToModify {boneName=thigh_armor_l_base}
    rotation {pitch=-5}
    translation=<5 -1>
   }
   n19 = "Transform (Modify) Bone - Bone: thigh_armor_r_base"!(ComponentPose=@n18, bAlphaBoolEnabled:bool=True, Translation="-2,0,0", Rotation="0,0,5", Scale="1,1,1") -> (Pose) {
    alphaInputType=Bool
    rotationMode=BMM_Additive
    rotationSpace=BCS_BoneSpace
    scaleSpace=BCS_BoneSpace
    translationMode=BMM_Additive
    translationSpace=BCS_BoneSpace
    alphaBoolBlend {blendOption=Linear}
    boneToModify {boneName=thigh_armor_r_base}
    rotation {roll=5}
    translation {x=-2}
   }
   n20 = "Transform (Modify) Bone - Bone: dyn_front_skirt_base"!(
    ComponentPose=Local To Component(LocalPose=Use cached pose 'Pose02'())
    Alpha:float=1
    Translation="-10,7,-4"
    Rotation="30,10,60"
    Scale="1,1,1"
   ) -> (Pose) {
    rotationMode=BMM_Additive
    rotationSpace=BCS_BoneSpace
    translationMode=BMM_Additive
    translationSpace=BCS_BoneSpace
    alphaBoolBlend {blendOption=Linear}
    boneToModify {boneName=dyn_front_skirt_base}
    rotation=<rot 30 10 60>
    translation=<-10 7 -4>
   }
   n21 = "Transform (Modify) Bone - Bone: thigh_armor_l_base"(ComponentPose=@n20, bAlphaBoolEnabled=True, Translation="1,-2,0", Rotation="0,-5,-10", Scale="1,1,1") {
    alphaInputType=Bool
    rotationMode=BMM_Additive
    rotationSpace=BCS_BoneSpace
    scaleSpace=BCS_BoneSpace
    translationMode=BMM_Additive
    translationSpace=BCS_BoneSpace
    alphaBoolBlend {blendOption=Linear}
    boneToModify {boneName=thigh_armor_l_base}
    rotation {roll=-10, yaw=-5}
    translation=<1 -2>
   }
   n22 = "Transform (Modify) Bone - Bone: thigh_armor_r_base"(ComponentPose=@n21, bAlphaBoolEnabled=True, Translation="2,2,1", Rotation="-5,0,0", Scale="1,1,1") {
    alphaInputType=Bool
    rotationMode=BMM_Additive
    rotationSpace=BCS_BoneSpace
    scaleSpace=BCS_BoneSpace
    translationMode=BMM_Additive
    translationSpace=BCS_BoneSpace
    alphaBoolBlend {blendOption=Linear}
    boneToModify {boneName=thigh_armor_r_base}
    rotation {pitch=-5}
    translation=<2 2 1>
   }
   n23 = RigidBody(ComponentPose=@n22) {
    overridePhysicsAsset=&F_MED_EmberRae_Physics
    simulationSpace=BaseBoneSpace
    alphaBoolBlend {blendOption=Linear}
   }
   n24 = "Constraint - dyn_bottom_skirt_01 (P)"(ComponentPose=@n23, Alpha=1, ConstraintWeights_0=1) {
    alphaBoolBlend {blendOption=Linear}
    boneToModify {boneName=dyn_bottom_skirt_01}
    constraintSetup [{targetBone {boneName=dyn_bottom_skirt_base}}]
    constraintWeights [1]
   }
   n25 = "Constraint - dyn_bottom_skirt_02 (P)"(ComponentPose=@n24, Alpha=0.5, ConstraintWeights_0=1) {
    alphaBoolBlend {blendOption=Linear}
    boneToModify {boneName=dyn_bottom_skirt_02}
    constraintSetup [{targetBone {boneName=dyn_bottom_skirt_01}}]
    constraintWeights [1]
   }
   n26 = "Constraint - dyn_bottom_skirt_03 (P)"(ComponentPose=@n25, Alpha=0.5, ConstraintWeights_0=1) {
    alphaBoolBlend {blendOption=Linear}
    boneToModify {boneName=dyn_bottom_skirt_03}
    constraintSetup [{targetBone {boneName=dyn_bottom_skirt_02}}]
    constraintWeights [1]
   }
   n27 = "Constraint - dyn_bottom_skirt_04 (P)"(ComponentPose=@n26, Alpha=1, ConstraintWeights_0=1) {
    alphaBoolBlend {blendOption=Linear}
    boneToModify {boneName=dyn_bottom_skirt_04}
    constraintSetup [{targetBone {boneName=dyn_bottom_skirt_03}}]
    constraintWeights [1]
   }
   n28 = Blend Poses by bool(
    BlendPose_0=Component To Local(ComponentPose=@n27)
    BlendPose_1=Blend Poses by bool(
     BlendPose_0=Component To Local(ComponentPose=@n19)
     BlendPose_1=Blend Poses by bool(BlendPose_0=@n13, BlendPose_1=@n9, bActiveValue=@n14) {blendType=Linear}
     BlendTime_0=0.6
     bActiveValue=Get IsJumpAir() -> (IsJumpAir:bool)
    ) {blendType=Linear}
    BlendTime_1=0.3
    bActiveValue=Get bIsCrouching()
   ) {blendType=Linear}
   n29 = "Transform (Modify) Bone - Bone: dyn_front_skirt_base"(
    ComponentPose=Local To Component(LocalPose=Use cached pose 'Pose02'())
    Alpha=1
    Translation="-20.270,7.000,-0.000"
    Rotation="0.000,90.000,-0.000"
    Scale="1,1,1"
   ) {
    rotationMode=BMM_Additive
    rotationSpace=BCS_BoneSpace
    translationMode=BMM_Additive
    translationSpace=BCS_BoneSpace
    alphaBoolBlend {blendOption=Linear}
    boneToModify {boneName=dyn_front_skirt_base}
    rotation {yaw=90}
    translation=<-20.27 7>
   }
   n30 = "Transform (Modify) Bone - Bone: dyn_bottom_skirt_base"!(ComponentPose=@n29, Alpha:float=1, Translation="2,0,-3", Rotation="-30,0,0", Scale="1,1,1") -> (Pose) {
    rotationMode=BMM_Additive
    rotationSpace=BCS_BoneSpace
    translationMode=BMM_Additive
    translationSpace=BCS_BoneSpace
    alphaBoolBlend {blendOption=Linear}
    boneToModify {boneName=dyn_bottom_skirt_base}
    rotation {pitch=-30}
    translation {x=2, z=-3}
   }
   n31 = "Transform (Modify) Bone - Bone: thigh_armor_l_base"(ComponentPose=@n30, bAlphaBoolEnabled=True, Translation="2,-2,0", Rotation="0,-5,-12", Scale="1,1,1") {
    alphaInputType=Bool
    rotationMode=BMM_Additive
    rotationSpace=BCS_BoneSpace
    scaleSpace=BCS_BoneSpace
    translationMode=BMM_Additive
    translationSpace=BCS_BoneSpace
    alphaBoolBlend {blendOption=Linear}
    boneToModify {boneName=thigh_armor_l_base}
    rotation {roll=-12, yaw=-5}
    translation=<2 -2>
   }
   n32 = "Transform (Modify) Bone - Bone: thigh_armor_r_base"(ComponentPose=@n31, bAlphaBoolEnabled=True, Translation="-2,-3,0", Rotation="-5,0,12", Scale="1,1,1") {
    alphaInputType=Bool
    rotationMode=BMM_Additive
    rotationSpace=BCS_BoneSpace
    scaleSpace=BCS_BoneSpace
    translationMode=BMM_Additive
    translationSpace=BCS_BoneSpace
    alphaBoolBlend {blendOption=Linear}
    boneToModify {boneName=thigh_armor_r_base}
    rotation {pitch=-5, roll=12}
    translation=<-2 -3>
   }
   n33 = RigidBody(ComponentPose=@n32) {
    overridePhysicsAsset=&F_MED_EmberRae_Physics
    simulationSpace=BaseBoneSpace
    alphaBoolBlend {blendOption=Linear}
   }
   n34 = Component To Local(ComponentPose=@n33)
   n35 = Get IsCrouchMoving() -> (IsCrouchMoving:bool)
   n36 = "Constraint - dyn_bottom_skirt_01 (P)"(
    ComponentPose=Local To Component(LocalPose=Blend Poses by bool(BlendPose_0=@n34, BlendPose_1=@n28, BlendTime_0=0.2, bActiveValue=@n35) {blendType=Linear})
    Alpha=Get SkirtAlphaForCape() -> (SkirtAlphaForCape:double)
    ConstraintWeights_0=1
   ) {
    alphaBoolBlend {blendOption=Linear}
    boneToModify {boneName=dyn_bottom_skirt_01}
    constraintSetup [{targetBone {boneName=dyn_bottom_skirt_base}}]
    constraintWeights [1]
   }
   n37 = "Constraint - dyn_bottom_skirt_02 (P)"(ComponentPose=@n36, Alpha=Get SkirtAlphaForCape(), ConstraintWeights_0=1) {
    alphaBoolBlend {blendOption=Linear}
    boneToModify {boneName=dyn_bottom_skirt_02}
    constraintSetup [{targetBone {boneName=dyn_bottom_skirt_01}}]
    constraintWeights [1]
   }
   n38 = "Constraint - dyn_bottom_skirt_03 (P)"(ComponentPose=@n37, Alpha=Get SkirtAlphaForCape(), ConstraintWeights_0=1) {
    alphaBoolBlend {blendOption=Linear}
    boneToModify {boneName=dyn_bottom_skirt_03}
    constraintSetup [{targetBone {boneName=dyn_bottom_skirt_02}}]
    constraintWeights [1]
   }
   n39 = "Constraint - dyn_bottom_skirt_04 (P)"(ComponentPose=@n38, Alpha=Get SkirtAlphaForCape(), ConstraintWeights_0=1) {
    alphaBoolBlend {blendOption=Linear}
    boneToModify {boneName=dyn_bottom_skirt_04}
    constraintSetup [{targetBone {boneName=dyn_bottom_skirt_03}}]
    constraintWeights [1]
   }
   n40 = "Transform (Modify) Bone - Bone: attach_backpack"!(
    ComponentPose=@n39
    bAlphaBoolEnabled:bool=Get IsWearingBackpack() -> (IsWearingBackpack:bool)
    Translation="0,0,0"
    Rotation="0,0,6"
    Scale="1,1,1"
   ) -> (Pose) {
    alphaInputType=Bool
    rotationMode=BMM_Additive
    rotationSpace=BCS_BoneSpace
    scaleSpace=BCS_BoneSpace
    translationMode=BMM_Additive
    translationSpace=BCS_BoneSpace
    alphaBoolBlend {blendOption=Linear}
    boneToModify {boneName=attach_backpack}
    rotation {roll=6}
   }
   n41 = Get bIsSkydiveDiveMode()
   n42 = "Transform (Modify) Bone - Bone: dyn_bottom_skirt_base"!(ComponentPose=@n40, bAlphaBoolEnabled:bool=@n41, Translation="0,0,0", Rotation="-10,0,0", Scale="1,1,1") -> (Pose) {
    alphaInputType=Bool
    rotationMode=BMM_Additive
    rotationSpace=BCS_BoneSpace
    scaleMode=BMM_Additive
    scaleSpace=BCS_BoneSpace
    translationMode=BMM_Additive
    translationSpace=BCS_BoneSpace
    alphaBoolBlend {blendInTime=0.2, blendOption=Linear, blendOutTime=0.1}
    boneToModify {boneName=dyn_bottom_skirt_base}
    rotation {pitch=-10}
   }
   n43 = "Transform (Modify) Bone - Bone: dyn_bottom_skirt_02"!(ComponentPose=@n42, bAlphaBoolEnabled:bool=@n41, Translation="0,0,0", Rotation="-30,0,0", Scale="1,1,1") -> (Pose) {
    alphaInputType=Bool
    rotationMode=BMM_Additive
    rotationSpace=BCS_BoneSpace
    scaleMode=BMM_Additive
    scaleSpace=BCS_BoneSpace
    translationMode=BMM_Additive
    translationSpace=BCS_BoneSpace
    alphaBoolBlend {blendInTime=0.2, blendOption=Linear, blendOutTime=0.1}
    boneToModify {boneName=dyn_bottom_skirt_02}
    rotation {pitch=-30}
   }
   n44 = "Transform (Modify) Bone - Bone: dyn_bottom_skirt_03"!(ComponentPose=@n43, bAlphaBoolEnabled:bool=@n41, Translation="0,0,0", Rotation="-40,0,0", Scale="1,1,1") -> (Pose) {
    alphaInputType=Bool
    rotationMode=BMM_Additive
    rotationSpace=BCS_BoneSpace
    scaleMode=BMM_Additive
    scaleSpace=BCS_BoneSpace
    translationMode=BMM_Additive
    translationSpace=BCS_BoneSpace
    alphaBoolBlend {blendInTime=0.2, blendOption=Linear, blendOutTime=0.1}
    boneToModify {boneName=dyn_bottom_skirt_03}
    rotation {pitch=-40}
   }
   n45 = "Transform (Modify) Bone - Bone: dyn_bottom_skirt_04"!(ComponentPose=@n44, bAlphaBoolEnabled:bool=@n41, Translation="0,0,0", Rotation="-50,0,0", Scale="1,1,1") -> (Pose) {
    alphaInputType=Bool
    rotationMode=BMM_Additive
    rotationSpace=BCS_BoneSpace
    scaleMode=BMM_Additive
    scaleSpace=BCS_BoneSpace
    translationMode=BMM_Additive
    translationSpace=BCS_BoneSpace
    alphaBoolBlend {blendInTime=0.2, blendOption=Linear, blendOutTime=0.1}
    boneToModify {boneName=dyn_bottom_skirt_04}
    rotation {pitch=-50}
   }
   Output Pose(Result=Component To Local(ComponentPose=@n45)) {layerGroup=None}
   n46 = Control Rig(Source=Copy Pose From Mesh() {bUseAttachedParent=true}, Alpha=0.8) {
    alphaBoolBlend {blendOption=Linear}
    controlRigAssetReference {blueprintRigClass=/Game/Characters/Player/Male/Medium/Bodies/M_MED_Boredom/Meshes/M_MED_DeformRig_Boredom_ControlRig.M_MED_DeformRig_Boredom_ControlRig_C}
   }
   n47 = "Constraint - dyn_front_skirt_base (PPPP)"!(
    ComponentPose=Local To Component(LocalPose=@n46)
    bAlphaBoolEnabled:bool=Get IsStraightPose() -> (IsStraightPose:bool)
    ConstraintWeights_0=0.4
    ConstraintWeights_1:float=0.4
    ConstraintWeights_2:float=0
    ConstraintWeights_3:float=0.2
   ) -> (Pose) {
    alphaInputType=Bool
    alphaBoolBlend {blendOption=Linear}
    boneToModify {boneName=dyn_front_skirt_base}
    constraintSetup [
     {targetBone {boneName=thigh_twist_01_l}}
     {targetBone {boneName=thigh_twist_01_r}}
     {targetBone {boneName=dyn_front_skirt_01}}
     {targetBone {boneName=pelvis}}
    ]
    constraintWeights [1,1,1,1]
   }
   n48 = "Transform (Modify) Bone - Bone: dyn_front_skirt_base"(ComponentPose=@n47, Alpha=1, Translation="-8.182,9.079,-0.400", Rotation="-1.739,7.508,-42.709", Scale="1.115,1.000,1.000") {
    rotationMode=BMM_Additive
    rotationSpace=BCS_BoneSpace
    scaleMode=BMM_Additive
    scaleSpace=BCS_BoneSpace
    translationMode=BMM_Additive
    translationSpace=BCS_BoneSpace
    alphaScaleBiasClamp {bInterpResult=true, interpSpeedDecreasing=1}
    boneToModify {boneName=dyn_front_skirt_base}
    rotation=<rot -1.739 7.508 -42.709>
    scale {x=1.115}
    translation=<-8.182 9.079 -0.4>
   }
   n49 = "Transform (Modify) Bone - Bone: front_skirt_extra"(ComponentPose=@n48, Alpha=1, Translation="-5.454,9.975,0.479", Rotation="14.562,-8.693,-30.797", Scale="1.000,1.000,1.910") {
    rotationMode=BMM_Additive
    rotationSpace=BCS_BoneSpace
    scaleMode=BMM_Additive
    scaleSpace=BCS_BoneSpace
    translationMode=BMM_Additive
    translationSpace=BCS_BoneSpace
    alphaScaleBiasClamp {bInterpResult=true}
    boneToModify {boneName=front_skirt_extra}
    rotation=<rot 14.562 -8.693 -30.797>
    scale {z=1.91}
    translation=<-5.454 9.975 0.479>
   }
   n50 = "Transform (Modify) Bone - Bone: front_skirt_extra"(ComponentPose=@n49, Alpha=1, Translation="-0.063,7.431,0.324", Rotation="12.287,-4.627,-51.261", Scale="1,1,1") {
    rotationMode=BMM_Additive
    rotationSpace=BCS_BoneSpace
    translationMode=BMM_Additive
    translationSpace=BCS_BoneSpace
    alphaScaleBiasClamp {bInterpResult=true}
    boneToModify {boneName=front_skirt_extra}
    rotation=<rot 12.287 -4.627 -51.261>
    translation=<-0.063 7.431 0.324>
   }
   n51 = "Transform (Modify) Bone - Bone: dyn_front_skirt_base"(ComponentPose=@n50, Alpha=1, Translation="0,2,0", Rotation="0,15,0", Scale="1,1,1") {
    rotationMode=BMM_Additive
    rotationSpace=BCS_BoneSpace
    translationMode=BMM_Additive
    translationSpace=BCS_BoneSpace
    alphaScaleBiasClamp {bInterpResult=true, interpSpeedDecreasing=1}
    boneToModify {boneName=dyn_front_skirt_base}
    rotation {yaw=15}
    translation {y=2}
   }
   n52 = "Transform (Modify) Bone - Bone: thigh_armor_l_base"!(ComponentPose=@n51, Alpha:float=1, Translation="-0.720,-0.010,-0.829", Rotation="5.476,-1.254,-5.483", Scale="1,1,1") -> (Pose) {
    rotationMode=BMM_Additive
    rotationSpace=BCS_BoneSpace
    scaleMode=BMM_Additive
    scaleSpace=BCS_BoneSpace
    translationMode=BMM_Additive
    translationSpace=BCS_BoneSpace
    alphaScaleBiasClamp {bInterpResult=true, interpSpeedDecreasing=1}
    boneToModify {boneName=thigh_armor_l_base}
    rotation=<rot 5.476 -1.254 -5.483>
    translation=<-0.72 -0.01 -0.829>
   }
   Save cached pose 'StartPose'(Pose=Component To Local(ComponentPose=@n52)) {cacheName=StartPose}
   n53 = "Transform (Modify) Bone - Bone: upperarm_l_armor_base"(
    ComponentPose=Local To Component(LocalPose=Use cached pose 'StartPose'())
    Alpha=Get Alpha_clavicleL_Y_plus() -> (Alpha_clavicleL_Y_plus:double)
    Translation="3,0,0"
    Rotation="0,0,0"
    Scale="1,1,1"
   ) {
    rotationMode=BMM_Additive
    rotationSpace=BCS_BoneSpace
    translationMode=BMM_Additive
    translationSpace=BCS_BoneSpace
    alphaBoolBlend {blendOption=Linear}
    boneToModify {boneName=upperarm_l_armor_base}
    translation {x=3}
   }
   n54 = "Transform (Modify) Bone - Bone: upperarm_l_armor_base"(
    ComponentPose=@n53
    Alpha=Get Alpha_upperarmL_Z_plus() -> (Alpha_upperarmL_Z_plus:double)
    Translation="0,-5,-1"
    Rotation="0,20,-40"
    Scale="1,1,1"
   ) {
    rotationMode=BMM_Additive
    rotationSpace=BCS_BoneSpace
    translationMode=BMM_Additive
    translationSpace=BCS_BoneSpace
    alphaBoolBlend {blendOption=Linear}
    boneToModify {boneName=upperarm_l_armor_base}
    rotation {roll=-40, yaw=20}
    translation {y=-5, z=-1}
   }
   n55 = "Transform (Modify) Bone - Bone: upperarm_l_armor_base"(
    ComponentPose=@n54
    Alpha=Get Alpha_upperarmL_Y_plus() -> (Alpha_upperarmL_Y_plus:double)
    Translation="6,0,0"
    Rotation="0,0,0"
    Scale="1,1,1"
   ) {
    rotationMode=BMM_Additive
    rotationSpace=BCS_BoneSpace
    translationMode=BMM_Additive
    translationSpace=BCS_BoneSpace
    alphaBoolBlend {blendOption=Linear}
    boneToModify {boneName=upperarm_l_armor_base}
    translation {x=6}
   }
   n56 = "Transform (Modify) Bone - Bone: dyn_bottom_skirt_base"(ComponentPose=@n55, bAlphaBoolEnabled=True, Translation="1,0,2", Rotation="0,0,0", Scale="1,1,1") {
    alphaInputType=Bool
    rotationMode=BMM_Additive
    rotationSpace=BCS_BoneSpace
    translationMode=BMM_Additive
    translationSpace=BCS_BoneSpace
    alphaBoolBlend {blendOption=Linear}
    boneToModify {boneName=dyn_bottom_skirt_base}
    translation {x=1, z=2}
   }
   n57 = "Transform (Modify) Bone - Bone: dyn_bottom_skirt_base"!(ComponentPose=@n56, Alpha:float=Get Alpha_thighR_Z_minus(), Translation="0,0,-1", Rotation="-10,0,0", Scale="1,1,1") -> (Pose) {
    rotationMode=BMM_Additive
    rotationSpace=BCS_BoneSpace
    translationMode=BMM_Additive
    translationSpace=BCS_BoneSpace
    alphaBoolBlend {blendOption=Linear}
    boneToModify {boneName=dyn_bottom_skirt_base}
    rotation {pitch=-10}
    translation {z=-1}
   }
   n58 = "Transform (Modify) Bone - Bone: dyn_bottom_skirt_base"(ComponentPose=@n57, Alpha=Get Alpha_thighL_Z_minus(), Translation="0,0,-1", Rotation="-10,0,0", Scale="1,1,1") {
    rotationMode=BMM_Additive
    rotationSpace=BCS_BoneSpace
    translationMode=BMM_Additive
    translationSpace=BCS_BoneSpace
    alphaBoolBlend {blendOption=Linear}
    boneToModify {boneName=dyn_bottom_skirt_base}
    rotation {pitch=-10}
    translation {z=-1}
   }
   n59 = "Transform (Modify) Bone - Bone: thigh_armor_l_base"!(ComponentPose=@n58, bAlphaBoolEnabled:bool=True, Translation="1,0,2", Rotation="-5,0,0", Scale="1,1,1") -> (Pose) {
    alphaInputType=Bool
    rotationMode=BMM_Additive
    rotationSpace=BCS_BoneSpace
    scaleSpace=BCS_BoneSpace
    translationMode=BMM_Additive
    translationSpace=BCS_BoneSpace
    alphaBoolBlend {blendOption=Linear}
    boneToModify {boneName=thigh_armor_l_base}
    rotation {pitch=-5}
    translation {x=1, z=2}
   }
   n60 = "Transform (Modify) Bone - Bone: thigh_armor_l_base"!(
    ComponentPose=@n59
    Alpha:float=Get Alpha_thighL_X_plus() -> (Alpha_thighL_X_plus:double)
    Translation="2,2,2"
    Rotation="0,10,20"
    Scale="1,1,1"
   ) -> (Pose) {
    rotationMode=BMM_Additive
    rotationSpace=BCS_BoneSpace
    scaleSpace=BCS_BoneSpace
    translationMode=BMM_Additive
    translationSpace=BCS_BoneSpace
    alphaBoolBlend {blendOption=Linear}
    boneToModify {boneName=thigh_armor_l_base}
    rotation {roll=20, yaw=10}
    translation=<2 2 2>
   }
   n61 = "Transform (Modify) Bone - Bone: thigh_armor_l_base"(ComponentPose=@n60, Alpha=Get Alpha_thighL_Y_plus() -> (Alpha_thighL_Y_plus:double), Translation="1,0,0", Rotation="0,-5,-10", Scale="1,1,1") {
    rotationMode=BMM_Additive
    rotationSpace=BCS_BoneSpace
    scaleSpace=BCS_BoneSpace
    translationMode=BMM_Additive
    translationSpace=BCS_BoneSpace
    alphaBoolBlend {blendOption=Linear}
    boneToModify {boneName=thigh_armor_l_base}
    rotation {roll=-10, yaw=-5}
    translation {x=1}
   }
   n62 = "Transform (Modify) Bone - Bone: thigh_armor_l_base"(ComponentPose=@n61, Alpha=Get Alpha_thighL_Z_minus(), Translation="3,0,5.5", Rotation="-15,15,15", Scale="1,1,1") {
    rotationMode=BMM_Additive
    rotationSpace=BCS_BoneSpace
    scaleSpace=BCS_BoneSpace
    translationMode=BMM_Additive
    translationSpace=BCS_BoneSpace
    alphaBoolBlend {blendOption=Linear}
    boneToModify {boneName=thigh_armor_l_base}
    rotation=<rot -15 15 15>
    translation {x=3, z=5.5}
   }
   n63 = "Transform (Modify) Bone - Bone: thigh_armor_r_base"(ComponentPose=@n62, bAlphaBoolEnabled=True, Translation="3,0,-3", Rotation="10,0,0", Scale="1,1,1") {
    alphaInputType=Bool
    rotationMode=BMM_Additive
    rotationSpace=BCS_BoneSpace
    scaleSpace=BCS_BoneSpace
    translationMode=BMM_Additive
    translationSpace=BCS_BoneSpace
    alphaBoolBlend {blendOption=Linear}
    boneToModify {boneName=thigh_armor_r_base}
    rotation {pitch=10}
    translation {x=3, z=-3}
   }
   n64 = "Transform (Modify) Bone - Bone: thigh_armor_r_base"!(
    ComponentPose=@n63
    Alpha:float=Get Alpha_thighR_X_plus() -> (Alpha_thighR_X_plus:double)
    Translation="1,2,-1"
    Rotation="0,10,-30"
    Scale="1,1,1"
   ) -> (Pose) {
    rotationMode=BMM_Additive
    rotationSpace=BCS_BoneSpace
    scaleSpace=BCS_BoneSpace
    translationMode=BMM_Additive
    translationSpace=BCS_BoneSpace
    alphaBoolBlend {blendOption=Linear}
    boneToModify {boneName=thigh_armor_r_base}
    rotation {roll=-30, yaw=10}
    translation=<1 2 -1>
   }
   n65 = "Transform (Modify) Bone - Bone: thigh_armor_r_base"(ComponentPose=@n64, Alpha=Get Alpha_thighR_Y_plus() -> (Alpha_thighR_Y_plus:double), Translation="3,0,-1.5", Rotation="0,0,0", Scale="1,1,1") {
    rotationMode=BMM_Additive
    rotationSpace=BCS_BoneSpace
    scaleSpace=BCS_BoneSpace
    translationMode=BMM_Additive
    translationSpace=BCS_BoneSpace
    alphaBoolBlend {blendOption=Linear}
    boneToModify {boneName=thigh_armor_r_base}
    translation {x=3, z=-1.5}
   }
   n66 = "Transform (Modify) Bone - Bone: thigh_armor_r_base"(ComponentPose=@n65, Alpha=Get Alpha_thighR_Z_minus(), Translation="5,-1,-3", Rotation="10,10,-10", Scale="1,1,1") {
    rotationMode=BMM_Additive
    rotationSpace=BCS_BoneSpace
    scaleSpace=BCS_BoneSpace
    translationMode=BMM_Additive
    translationSpace=BCS_BoneSpace
    alphaBoolBlend {blendOption=Linear}
    boneToModify {boneName=thigh_armor_r_base}
    rotation=<rot 10 10 -10>
    translation=<5 -1 -3>
   }
   Save cached pose 'Pose02'(Pose=Component To Local(ComponentPose=@n66)) {cacheName=Pose02}
  ]}
 EventGraph {nodes [
   n1 = Event BlueprintUpdateAnimation()
   > n2 = Sequence!(execute) -> (then_0, then_1, then_2:Exec, then_3:Exec, then_4:Exec)
   > Set IsCrouchMoving(IsCrouchMoving:bool=AND Boolean(A=Get bIsCrouching(), B=Get bIsMoving2D())) -> (Output_Get:bool)
   n3 = Get bIsSkyDiving()
   n4 = Get bIsJumping()
   n5 = Get bIsFalling()
   > Set IsJumpAir(IsJumpAir:bool=AND Boolean(A=OR Boolean(A=@n4, B=@n5), B=NOT Boolean(A=@n3))) -> (Output_Get:bool)
   n6 = Get IsJumpAir()
   n7 = Get bIsCrouching()
   > Set IsStraightPose(IsStraightPose:bool=NOT Boolean(A=OR Boolean(A=@n6, B=@n7))) -> (Output_Get:bool)
   n8 = Get bIsCrouching()
   n9 = OR Boolean(A=Get bIsJumping(), B=Get bIsFalling())
   > Set IsStreightMoving(IsStreightMoving:bool=AND Boolean!(A=Get bIsAccelerating2D(), B=NOT Boolean(A=@n8), C:bool=NOT Boolean(A=@n9)) -> (ReturnValue)) -> (Output_Get:bool)
   n10 = GetAnimInstance(self=GetOwnerSkeletalMeshComponent())
   n11 = Montage_IsPlaying(self=@n10, Montage=/BRCosmetics/Animation/Game/MainPlayer/Emotes/GoldenGlory/CMF/Emote_GoldenGlory_CMF_M)
   n12 = Montage_IsPlaying(self=@n10, Montage=/BRCosmetics/Animation/Game/MainPlayer/Emotes/GoldenGlory/CMF/Emote_GoldenGlory_CMF_FrontEnd_M)
   > Set GoldenGloryEmote(GoldenGloryEmote:double="To Float (Boolean)"(InBool=OR Boolean(A=@n11, B=@n12))) -> (Output_Get:double)
   Is Valid(exec=@n2.then_1, InputObject=Get PlayerPawn() -> (PlayerPawn:Fort Player Pawn Athena Object Reference))
   > Set WPO Globals Function(
    Material=Get Mat Instance Dynamic()
    Overlay Material=Get Overlay Mat Instance Dynamic()
    Skeletal Mesh=Get SkelMesh()
    "Local Noise Transform  Location"=Get Local Noise Transform Location() -> (Local Noise Transform Location:Vector)
    Local Noise Offset=Get Local Noise Offset()
   )
   > Set WPO Regional Function!(
    execute
    Material:Material Instance Dynamic Object Reference=Get Mat Instance Dynamic()
    Overlay Material:Material Instance Dynamic Object Reference=Get Overlay Mat Instance Dynamic()
    Skeletal Mesh:Skeletal Mesh Component Object Reference=Get SkelMesh()
    Velocity Transform:Name=foot_r
    Velocity Interp Speed:double=Get Velocity Interp Speed()
    Collision Transform:Name=calf_r
    Collision Transform Axis:EAxis Enum
    Invert Collision Transform Axis:bool=false
    Delta Time X:double=@n1.DeltaTimeX
    Previous Velocity Transform Location:"Vector (by ref)"=Get R1 Velocity Transform Location()
    Previous Velocity Transform Velocity:"Vector (by ref)"=Get R1 Velocity Transform Velocity()
    Region:int=1
   ) -> (then)
   > Set WPO Regional Function(
    Material=Get Mat Instance Dynamic()
    Overlay Material=Get Overlay Mat Instance Dynamic()
    Skeletal Mesh=Get SkelMesh()
    Velocity Transform=foot_l
    Velocity Interp Speed=Get Velocity Interp Speed()
    Collision Transform=calf_l
    Invert Collision Transform Axis=true
    Delta Time X=@n1.DeltaTimeX
    Previous Velocity Transform Location=Get R2 Velocity Transform Location()
    Previous Velocity Transform Velocity=Get R2 Velocity Transform Velocity()
    Region=2
   )
   > n13 = Branch(Condition=Get bIsSkydiveFloating())
   > n14 = Set WPO Pho Wind Intensity(WPO Pho Wind Intensity=1)
   SetScalarParameterValue(execute=[@n15.then, @n14.then], self=Get Mat Instance Dynamic(), ParameterName=ClothSim_WindStrength, Value=Get WPO Pho Wind Intensity())
   n15 = Set WPO Pho Wind Intensity(execute=@n13.else, WPO Pho Wind Intensity=0)
   n16 = Get SkelMesh()
   Set Alpha_upperarmL_Z_plus(
    execute=@n2.then_2
    Alpha_upperarmL_Z_plus:double=GetAlphaFromReferenceTransformDelta(SkeletalMeshComponent=@n16, BoneName=upperarm_l, BaseName=clavicle_l, InputRange=(X=30,Y=80), TwistAxis="1,0,0")
   ) -> (Output_Get:double)
   > Set Alpha_upperarmL_Y_plus(Alpha_upperarmL_Y_plus:double=GetAlphaFromReferenceTransformDelta(SkeletalMeshComponent=@n16, BoneName=upperarm_l, BaseName=clavicle_l, DeltaComponentType=EulerY, InputRange=(X=10,Y=80), TwistAxis="1,0,0")) -> (Output_Get:double)
   > Set Alpha_clavicleL_Y_plus(Alpha_clavicleL_Y_plus:double=GetAlphaFromReferenceTransformDelta(SkeletalMeshComponent=@n16, BoneName=clavicle_l, BaseName=spine_05, DeltaComponentType=EulerY, InputRange=(X=10,Y=60), TwistAxis="1,0,0")) -> (Output_Get:double)
   > Set Detected Leg position GoldenGlory(Detected Leg position GoldenGlory:double=float * float!(
     A=GetAlphaFromReferenceTransformDelta(SkeletalMeshComponent=@n16, BoneName=thigh_r, BaseName=pelvis, DeltaComponentType=EulerY, InputRange=(X=0,Y=-20), TwistAxis="1,0,0")
     B=GetAlphaFromReferenceTransformDelta(SkeletalMeshComponent=@n16, BoneName=upperarm_r, BaseName=spine_05, DeltaComponentType=EulerY, InputRange=(X=20,Y=40), TwistAxis="1,0,0")
     C:double=Get GoldenGloryEmote() -> (GoldenGloryEmote:double)
    ) -> (ReturnValue) {is=title K2Node_PromotableOperator}) -> (Output_Get:double)
   n17 = Get SkelMesh()
   Set Alpha_thighL_Z_minus(
    execute=@n2.then_3
    Alpha_thighL_Z_minus:double=GetAlphaFromReferenceTransformDelta(SkeletalMeshComponent=@n17, BoneName=thigh_l, BaseName=pelvis, InputRange=(X=10,Y=-80), TwistAxis="1,0,0")
   ) -> (Output_Get:double)
   > Set Alpha_thighL_Y_plus(Alpha_thighL_Y_plus:double=GetAlphaFromReferenceTransformDelta(SkeletalMeshComponent=@n17, BoneName=thigh_l, BaseName=pelvis, DeltaComponentType=EulerY, InputRange=(X=10,Y=30), TwistAxis="1,0,0")) -> (Output_Get:double)
   > Set Alpha_thighL_X_plus(Alpha_thighL_X_plus:double=GetAlphaFromReferenceTransformDelta(SkeletalMeshComponent=@n17, BoneName=thigh_l, BaseName=pelvis, DeltaComponentType=EulerX, InputRange=(X=5,Y=40), TwistAxis="1,0,0")) -> (Output_Get:double)
   n18 = Get SkelMesh()
   Set Alpha_thighR_Z_minus(
    execute=@n2.then_4
    Alpha_thighR_Z_minus:double=GetAlphaFromReferenceTransformDelta(SkeletalMeshComponent=@n18, BoneName=thigh_r, BaseName=pelvis, InputRange=(X=10,Y=-80), TwistAxis="1,0,0")
   ) -> (Output_Get:double)
   > Set Alpha_thighR_X_plus(Alpha_thighR_X_plus:double=GetAlphaFromReferenceTransformDelta(SkeletalMeshComponent=@n18, BoneName=thigh_r, BaseName=pelvis, DeltaComponentType=EulerX, InputRange=(X=17,Y=45), TwistAxis="1,0,0")) -> (Output_Get:double)
   > Set Alpha_thighR_Y_plus(Alpha_thighR_Y_plus:double=GetAlphaFromReferenceTransformDelta(SkeletalMeshComponent=@n18, BoneName=thigh_r, BaseName=pelvis, DeltaComponentType=EulerY, InputRange=(X=10,Y=30), TwistAxis="1,0,0")) -> (Output_Get:double)
   Event BlueprintInitializeAnimation()
   > Set SkelMesh(SkelMesh=GetOwnerSkeletalMeshComponent())
   n19 = TryGetPawnOwner()
   > Is Valid(InputObject=@n19)
   > n20 = Cast To FortPlayerPawnAthena(Object=@n19)
   > Set PlayerPawn(PlayerPawn:Fort Player Pawn Athena Object Reference=@n20) -> (Output_Get:Fort Player Pawn Athena Object Reference)
   Event OnCharacterCustomizationLoaded()
   > n21 = Set SkelMesh(SkelMesh=GetPartSkeletalMeshComponent())
   > n22 = Get Dynamic Material Instance(SkeletalMeshComponent=@n21)
   > Set Mat Instance Dynamic(Mat Instance Dynamic=@n22)
   > n23 = GetOrCreateDynamicOverlayMaterial(MeshComponent=@n21, MaterialName=None)
   > Set Overlay Mat Instance Dynamic(Overlay Mat Instance Dynamic=@n23)
   n24 = CheckForBackbling()
   > Set IsWearingBackpack(IsWearingBackpack:bool=AND Boolean!(A=@n24, B=NOT Boolean(A=@n24.bIsBackblingACape)) -> (ReturnValue)) -> (Output_Get:bool)
   > n25 = Set IsWearingCape(IsWearingCape:bool=@n24.bIsBackblingACape) -> (Output_Get:bool)
   > n26 = Branch(Condition=@n25)
   > Set SkirtAlphaForCape(SkirtAlphaForCape:double=0.5) -> (Output_Get:double)
   Set SkirtAlphaForCape(execute=@n26.else, SkirtAlphaForCape=0)
   Reroute Node(InputPin:float=@n1.DeltaTimeX) -> (OutputPin:float)
  ]}
]
variables [
 Alpha_clavicleL_Y_plus {type=number}
 Alpha_thighL_X_plus {type=number}
 Alpha_thighL_Y_plus {type=number}
 Alpha_thighL_Z_minus {type=number}
 Alpha_thighR_X_plus {type=number}
 Alpha_thighR_Y_plus {type=number}
 Alpha_thighR_Z_minus {type=number}
 Alpha_upperarmL_Y_plus {type=number}
 Alpha_upperarmL_Z_plus {type=number}
 Detected Leg position GoldenGlory {type=number}
 GoldenGloryEmote {type=number}
 IsCrouchMoving {type=boolean}
 IsJumpAir {type=boolean}
 IsStartMove {type=boolean}
 IsStraightPose {type=boolean}
 IsStreightMoving {type=boolean}
 IsWearingBackpack {type=boolean}
 IsWearingCape {type=boolean}
 Local Noise Offset {type=Vector}
 Local Noise Transform Location {type=Vector}
 Mat Instance Dynamic {type=MaterialInstanceDynamic}
 Overlay Mat Instance Dynamic {type=MaterialInstanceDynamic}
 PlayerPawn {type=FortPlayerPawnAthena}
 R1 Velocity Transform Location {type=Vector}
 R1 Velocity Transform Velocity {type=Vector}
 R2 Velocity Transform Location {type=Vector}
 R2 Velocity Transform Velocity {type=Vector}
 SkelMesh {type=SkeletalMeshComponent}
 SkirtAlphaForCape {type=number}
 Velocity Interp Speed {type=number, default=30}
 VelocityAlpha {type=number}
 WPO Pho Wind Intensity {type=number}
]
