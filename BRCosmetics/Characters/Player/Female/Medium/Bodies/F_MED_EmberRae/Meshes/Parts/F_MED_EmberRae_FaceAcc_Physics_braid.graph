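# Physics asset for the hair braid on the EmberRae female character's face accessory mesh, providing bodies and joint constraints for the braid's motion.
# A sphere and four capsules form the braid chain from the head, with tight twist limits and swing range that narrows toward the tip.
# Capsules on head, spine, arms and thighs supply body collision shapes.
# Spine and arm joints rotate freely, thighs are unconstrained, and the spine_05 to upperarm_l constraint appears twice.
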PhysicsAsset /BRCosmetics/Characters/Player/Female/Medium/Bodies/F_MED_EmberRae/Meshes/Parts/F_MED_EmberRae_FaceAcc_Physics_braid
bodies {
 dyn_hair_braid_2 [{shapeType=Sphere, radius=4.0044, shapeName=dyn_hair_braid_2_sphere, center=<6.964 -2.4653 0.3244>}]
 dyn_hair_braid_3 [{shapeType=Capsule, length=7, radius=3.5, shapeName=dyn_hair_braid_3_capsule, center=<6 -1.1315 0>, rotation=<rot 90 0 0>}]
 dyn_hair_braid_4 [{shapeType=Capsule, length=6, radius=3.5, shapeName=dyn_hair_braid_4_capsule, center=<6 -0.791 0>, rotation=<rot 90 0 0>}]
 dyn_hair_braid_5 [{shapeType=Capsule, length=3.5, radius=3, shapeName=dyn_hair_braid_5_capsule, center=<6 0 0>, rotation=<rot 90 0 0>}]
 dyn_hair_braid_6 [{shapeType=Capsule, length=10.25, radius=5.75, shapeName=dyn_hair_braid_6_capsule, center=<9.2846 -4.3662 0>, rotation=<rot 90 0 14.9995>}]
 head [
  {shapeType=Capsule, length=9, radius=6, shapeName=head_capsule, center=<-7.7586 2.2248 0>, rotation=<rot 90 0 -164>}
  {shapeType=Capsule, length=0, radius=17.9005, shapeName=head_capsule, center=<7.6016 6.857 0>, rotation=<rot -90 179.789 -179.7891>}
 ]
 lowerarm_l [{shapeType=Capsule, length=15, radius=7, shapeName=lowerarm_l_capsule, center=<13.4664 0 0>, rotation=<rot 90 0 0>}]
 lowerarm_r [{shapeType=Capsule, length=15, radius=7, shapeName=lowerarm_r_capsule, center=<-12.8723 0 0>, rotation=<rot 90 0 0>}]
 spine_01 [{shapeType=Capsule, length=3.25, radius=20, shapeName=spine_01_capsule, center=<-15.042 4.1419 0>}]
 spine_02 [{shapeType=Capsule, length=7, radius=11, shapeName=spine_02_capsule, center=<0 3.3725 0>}]
 spine_03 [{shapeType=Capsule, length=7, radius=11, shapeName=spine_03_capsule, center=<0 3.3419 0>}]
 spine_04 [{shapeType=Capsule, length=5.25, radius=11, shapeName=spine_04_capsule, center=<0 3.4218 0>}]
 spine_05 [{shapeType=Capsule, length=15, radius=10.75, shapeName=spine_05_capsule, center=<0 2.2842 0>}]
 thigh_l [{shapeType=Capsule, length=8.5, radius=7.75, shapeName=thigh_l_capsule, center=<6.7234 -1.2135 -1.1118>, rotation=<rot -70 -0.001 90.0002>}]
 thigh_r [{shapeType=Capsule, length=7.5, radius=7.75, shapeName=thigh_l_capsule, center=<-7.4204 1.1233 -0.0844>, rotation=<rot -70 -0.001 -89.9998>}]
 upperarm_l [{shapeType=Capsule, length=24, radius=6.5, shapeName=upperarm_l_capsule, center=<12.8495 0 0>, rotation=<rot 90 0 0>}]
 upperarm_r [{shapeType=Capsule, length=24, radius=6.5, shapeName=upperarm_l_capsule, center=<-13.4099 0 0>, rotation=<rot 90 0 180>}]
}
constraints [
 {
  bone1Name=spine_01
  bone2Name=spine_02
  swing1LimitDegrees=45
  swing1Motion=Free
  swing2LimitDegrees=45
  swing2Motion=Free
  twistLimitDegrees=45
  twistMotion=Free
 }
 {
  bone1Name=spine_02
  bone2Name=spine_03
  swing1LimitDegrees=45
  swing1Motion=Free
  swing2LimitDegrees=45
  swing2Motion=Free
  twistLimitDegrees=45
  twistMotion=Free
 }
 {
  bone1Name=spine_03
  bone2Name=spine_04
  swing1LimitDegrees=45
  swing1Motion=Free
  swing2LimitDegrees=45
  swing2Motion=Free
  twistLimitDegrees=45
  twistMotion=Free
 }
 {
  bone1Name=spine_04
  bone2Name=spine_05
  swing1LimitDegrees=45
  swing1Motion=Free
  swing2LimitDegrees=45
  swing2Motion=Free
  twistLimitDegrees=45
  twistMotion=Free
 }
 {
  bone1Name=spine_05
  bone2Name=upperarm_l
  swing1LimitDegrees=45
  swing1Motion=Free
  swing2LimitDegrees=45
  swing2Motion=Free
  twistLimitDegrees=45
  twistMotion=Free
 }
 {
  bone1Name=spine_05
  bone2Name=upperarm_l
  swing1LimitDegrees=45
  swing1Motion=Free
  swing2LimitDegrees=45
  swing2Motion=Free
  twistLimitDegrees=45
  twistMotion=Free
 }
 {
  bone1Name=spine_05
  bone2Name=upperarm_r
  swing1LimitDegrees=45
  swing1Motion=Free
  swing2LimitDegrees=45
  swing2Motion=Free
  twistLimitDegrees=45
  twistMotion=Free
 }
 {bone1Name=dyn_hair_braid_3, bone2Name=dyn_hair_braid_2, swing1LimitDegrees=120, swing2LimitDegrees=60, twistLimitDegrees=5}
 {bone1Name=dyn_hair_braid_4, bone2Name=dyn_hair_braid_3, swing1LimitDegrees=120, swing2LimitDegrees=60, twistLimitDegrees=3}
 {bone1Name=dyn_hair_braid_5, bone2Name=dyn_hair_braid_4, swing1LimitDegrees=80, swing2LimitDegrees=60, twistLimitDegrees=3}
 {bone1Name=dyn_hair_braid_6, bone2Name=dyn_hair_braid_5, swing1LimitDegrees=65, swing2LimitDegrees=60, twistLimitDegrees=3}
 {bone1Name=dyn_hair_braid_2, bone2Name=head, swing1LimitDegrees=120, swing2LimitDegrees=20, twistLimitDegrees=10}
 {
  bone1Name=lowerarm_r
  bone2Name=upperarm_r
  swing1LimitDegrees=45
  swing1Motion=Free
  swing2LimitDegrees=45
  swing2Motion=Free
  twistLimitDegrees=45
  twistMotion=Free
 }
 {
  bone1Name=lowerarm_l
  bone2Name=upperarm_l
  swing1LimitDegrees=45
  swing1Motion=Free
  swing2LimitDegrees=45
  swing2Motion=Free
  twistLimitDegrees=45
  twistMotion=Free
 }
]
