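# Physics asset for the hair fringe in the female EmberRae character's face accessory mesh, defining collision bodies and joint limits.
# Five two-segment strands use small capsules, while a sphere, capsule and two boxes approximate the head.
# Strands 1, 2, 4 and 5 chain from the head with narrow twist and mostly tight swing limits.
# Both strand 3 segments constrain to a base sphere, which has no head constraint, and one swings much wider.
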PhysicsAsset /BRCosmetics/Characters/Player/Female/Medium/Bodies/F_MED_EmberRae/Meshes/Parts/F_MED_EmberRae_FaceAcc_Physics_fringe
bodies {
 dyn_hair_fr_1_1 [{
   shapeType=Capsule
   length=2
   radius=1
   shapeName=dyn_hair_fr_1_1_capsule
   center=<0.8925 -0.0659 0.0745>
   rotation=<rot 84.9999 -0.0009 -0.001>
  }]
 dyn_hair_fr_1_2 [{
   shapeType=Capsule
   length=4
   radius=1
   shapeName=dyn_hair_fr_1_2_capsule
   center=<1.2296 -0.0036 -1.1746>
   rotation=<rot 59.9999 -0.0004 -0.0003>
  }]
 dyn_hair_fr_2_1 [{shapeType=Capsule, length=2, radius=0.5, shapeName=dyn_hair_fr_2_1_capsule, center=<2.1714 0.0448 0>, rotation=<rot 90 0 10.0002>}]
 dyn_hair_fr_2_2 [{shapeType=Capsule, length=3, radius=0.5, shapeName=dyn_hair_fr_2_2_capsule, center=<0.9873 -0.8738 0>, rotation=<rot 90 180 -154.9994>}]
 dyn_hair_fr_3_1 [{
   shapeType=Capsule
   length=4
   radius=1
   shapeName=dyn_hair_fr_3_1_capsule
   center=<1.4459 -0.0145 -0.3177>
   rotation=<rot 79.9999 15.0005 0.0002>
  }]
 dyn_hair_fr_3_2 [{shapeType=Capsule, length=2.25, radius=1, shapeName=dyn_hair_fr_3_2_capsule, center=<1 0 -0.7903>, rotation=<rot 90 0 0>}]
 dyn_hair_fr_3_base [{shapeType=Sphere, radius=1, shapeName=dyn_hair_fr_3_base_sphere}]
 dyn_hair_fr_4_1 [{shapeType=Capsule, length=1, radius=0.7, shapeName=dyn_hair_fr_4_1_capsule, center=<1.0758 0 -0.9616>, rotation=<rot 90 0 0>}]
 dyn_hair_fr_4_2 [{
   shapeType=Capsule
   length=3
   radius=0.7
   shapeName=dyn_hair_fr_4_2_capsule
   center=<2.0588 0.3836 0.754>
   rotation=<rot 70.0196 -177.4626 172.4603>
  }]
 dyn_hair_fr_5_1 [{shapeType=Capsule, length=2, radius=0.5, shapeName=dyn_hair_fr_5_1_capsule, center=<2.24 0.1594 0>, rotation=<rot 90 0 0>}]
 dyn_hair_fr_5_2 [{
   shapeType=Capsule
   length=4
   radius=0.7
   shapeName=dyn_hair_fr_5_2_capsule
   center=<1.9782 -1.6642 0.5479>
   rotation=<rot 75 140.0003 179.9999>
  }]
 head [
  {shapeType=Sphere, radius=10.75, shapeName=head_sphere, center=<8.4788 -0.5375 0>}
  {shapeType=Capsule, length=10.25, radius=7, shapeName=head_capsule, center=<4.4603 0 0>, rotation=<rot 90 -6.2419 -11.2412>}
  {shapeType=Box, extentX=6.5, extentY=2.5, extentZ=6.5, shapeName=head_box, center=<7.559 10.3682 -5.1675>, rotation=<rot 0 -0.0004 30>}
  {shapeType=Box, extentX=6.5, extentY=2, extentZ=6.5, shapeName=head_box, center=<7.0177 9.9362 5.4005>, rotation=<rot 0 -0.0013 -25>}
 ]
}
constraints [
 {bone1Name=dyn_hair_fr_3_1, bone2Name=dyn_hair_fr_3_base, swing1LimitDegrees=20, swing2LimitDegrees=20, twistLimitDegrees=2}
 {bone1Name=dyn_hair_fr_1_2, bone2Name=dyn_hair_fr_1_1, swing1LimitDegrees=5, swing2LimitDegrees=5, twistLimitDegrees=5}
 {bone1Name=dyn_hair_fr_2_2, bone2Name=dyn_hair_fr_2_1, swing1LimitDegrees=5, swing2LimitDegrees=5, twistLimitDegrees=5}
 {bone1Name=dyn_hair_fr_1_1, bone2Name=head, swing1LimitDegrees=5, swing2LimitDegrees=5, twistLimitDegrees=2}
 {bone1Name=dyn_hair_fr_2_1, bone2Name=head, swing1LimitDegrees=5, swing2LimitDegrees=5, twistLimitDegrees=2}
 {bone1Name=dyn_hair_fr_4_1, bone2Name=head, swing1LimitDegrees=10, swing2LimitDegrees=10, twistLimitDegrees=2}
 {bone1Name=dyn_hair_fr_4_2, bone2Name=dyn_hair_fr_4_1, swing1LimitDegrees=20, swing2LimitDegrees=40, twistLimitDegrees=2}
 {bone1Name=dyn_hair_fr_5_2, bone2Name=dyn_hair_fr_5_1, swing1LimitDegrees=15, swing2LimitDegrees=15, twistLimitDegrees=2}
 {bone1Name=dyn_hair_fr_5_1, bone2Name=head, swing1LimitDegrees=5, swing2LimitDegrees=5, twistLimitDegrees=2}
 {bone1Name=dyn_hair_fr_3_2, bone2Name=dyn_hair_fr_3_base, swing1LimitDegrees=70, swing2LimitDegrees=63, twistLimitDegrees=2}
]
