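# Physics asset for the braid on EmberRae's face accessory mesh, named for use during skydiving.
# Five capsule bodies form a braid chain from the head, each joint allowing a wider swing on one axis and tight limits on the others.
# Capsules on the head, spine, pelvis, arms and thighs approximate the body, presumably as braid colliders.
# These body parts connect to neighbors or the pelvis through angularly free constraints.
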PhysicsAsset /BRCosmetics/Characters/Player/Female/Medium/Bodies/F_MED_EmberRae/Meshes/Parts/F_MED_EmberRae_FaceAcc_Skydive_Physics_braid
bodies {
 dyn_hair_braid_2 [{shapeType=Capsule, length=5.8475, radius=3.9025, shapeName=dyn_hair_braid_2_capsule, center=<7.4996 -3 0>, rotation=<rot 90 0 0>}]
 dyn_hair_braid_3 [{shapeType=Capsule, length=3.8475, radius=6.1525, shapeName=dyn_hair_braid_3_capsule, center=<6.1623 -3 0>, rotation=<rot 90 0 0>}]
 dyn_hair_braid_4 [{shapeType=Capsule, length=3.8475, radius=6.1525, shapeName=dyn_hair_braid_4_capsule, center=<6.1623 -3 0>, rotation=<rot 90 0 0>}]
 dyn_hair_braid_5 [{shapeType=Capsule, length=3.8475, radius=6.1525, shapeName=dyn_hair_braid_5_capsule, center=<6.1623 -3 0>, rotation=<rot 90 0 0>}]
 dyn_hair_braid_6 [{shapeType=Capsule, length=10.622, radius=5.9025, shapeName=dyn_hair_braid_6_capsule, center=<10.5673 -3 0>, rotation=<rot 90 0 0>}]
 head [
  {shapeType=Capsule, length=9, radius=6, shapeName=head_capsule, center=<-7.7586 2.2248 0>, rotation=<rot 90 0 -164>}
  {shapeType=Capsule, length=0, radius=18.5, shapeName=head_capsule, center=<7.6016 8.0274 0>, rotation=<rot -90 172.4476 -172.4478>}
 ]
 lowerarm_l [{shapeType=Capsule, length=15, radius=7, shapeName=lowerarm_l_capsule, center=<13.4664 0 0>, rotation=<rot 90 0 0>}]
 lowerarm_r [{shapeType=Capsule, length=15, radius=7, shapeName=lowerarm_r_capsule, center=<-12.8723 0 0>, rotation=<rot 90 0 0>}]
 pelvis [{shapeType=Capsule, length=25.426, radius=20, shapeName=pelvis_capsule, center=<0 5.0804 0>, rotation=<rot 0 -90 0>}]
 spine_01 [{shapeType=Capsule, length=45.7448, radius=25.2906, shapeName=spine_01_capsule, center=<-15.042 10.0488 0>}]
 spine_02 [{shapeType=Capsule, length=45.9329, radius=16.6208, shapeName=spine_02_capsule, center=<0 6.0049 0>}]
 spine_03 [{shapeType=Capsule, length=40.8154, radius=11, shapeName=spine_03_capsule, center=<0 3.3419 0>}]
 spine_04 [{shapeType=Capsule, length=5.25, radius=11, shapeName=spine_04_capsule, center=<0 3.4218 0>}]
 spine_05 [{shapeType=Capsule, length=29.0988, radius=10.75, shapeName=spine_05_capsule, center=<0 2.2842 0>}]
 thigh_l [{shapeType=Capsule, length=8.5, radius=7.75, shapeName=thigh_l_capsule, center=<6.7234 -1.2135 -1.1118>, rotation=<rot -70 -0.001 90.0002>}]
 thigh_r [{shapeType=Capsule, length=7.5, radius=7.75, shapeName=thigh_l_capsule, center=<-7.4204 1.1233 -0.0844>, rotation=<rot -70 -0.001 -89.9998>}]
 upperarm_l [{shapeType=Capsule, length=24, radius=6.5, shapeName=upperarm_l_capsule, center=<12.8495 0 0>, rotation=<rot 90 0 0>}]
 upperarm_r [{shapeType=Capsule, length=24, radius=6.5, shapeName=upperarm_l_capsule, center=<-13.4099 0 0>, rotation=<rot 90 0 180>}]
}
constraints [
 {
  bone1Name=spine_01
  bone2Name=spine_02
  swing1LimitDegrees=45
  swing1Motion=Free
  swing2LimitDegrees=45
  swing2Motion=Free
  twistLimitDegrees=45
  twistMotion=Free
 }
 {
  bone1Name=spine_02
  bone2Name=spine_03
  swing1LimitDegrees=45
  swing1Motion=Free
  swing2LimitDegrees=45
  swing2Motion=Free
  twistLimitDegrees=45
  twistMotion=Free
 }
 {
  bone1Name=spine_03
  bone2Name=spine_04
  swing1LimitDegrees=45
  swing1Motion=Free
  swing2LimitDegrees=45
  swing2Motion=Free
  twistLimitDegrees=45
  twistMotion=Free
 }
 {
  bone1Name=spine_04
  bone2Name=spine_05
  swing1LimitDegrees=45
  swing1Motion=Free
  swing2LimitDegrees=45
  swing2Motion=Free
  twistLimitDegrees=45
  twistMotion=Free
 }
 {
  bone1Name=spine_05
  bone2Name=upperarm_l
  swing1LimitDegrees=45
  swing1Motion=Free
  swing2LimitDegrees=45
  swing2Motion=Free
  twistLimitDegrees=45
  twistMotion=Free
 }
 {
  bone1Name=spine_05
  bone2Name=upperarm_l
  swing1LimitDegrees=45
  swing1Motion=Free
  swing2LimitDegrees=45
  swing2Motion=Free
  twistLimitDegrees=45
  twistMotion=Free
 }
 {
  bone1Name=spine_05
  bone2Name=upperarm_r
  swing1LimitDegrees=45
  swing1Motion=Free
  swing2LimitDegrees=45
  swing2Motion=Free
  twistLimitDegrees=45
  twistMotion=Free
 }
 {
  bone1Name=lowerarm_r
  bone2Name=upperarm_r
  swing1LimitDegrees=45
  swing1Motion=Free
  swing2LimitDegrees=45
  swing2Motion=Free
  twistLimitDegrees=45
  twistMotion=Free
 }
 {
  bone1Name=lowerarm_l
  bone2Name=upperarm_l
  swing1LimitDegrees=45
  swing1Motion=Free
  swing2LimitDegrees=45
  swing2Motion=Free
  twistLimitDegrees=45
  twistMotion=Free
 }
 {bone1Name=dyn_hair_braid_2, bone2Name=head, swing1LimitDegrees=40, swing2LimitDegrees=5, twistLimitDegrees=5}
 {bone1Name=dyn_hair_braid_3, bone2Name=dyn_hair_braid_2, swing1LimitDegrees=40, swing2LimitDegrees=5, twistLimitDegrees=5}
 {bone1Name=dyn_hair_braid_4, bone2Name=dyn_hair_braid_3, swing1LimitDegrees=40, swing2LimitDegrees=5, twistLimitDegrees=5}
 {bone1Name=dyn_hair_braid_5, bone2Name=dyn_hair_braid_4, swing1LimitDegrees=40, swing2LimitDegrees=5, twistLimitDegrees=5}
 {bone1Name=dyn_hair_braid_6, bone2Name=dyn_hair_braid_5, swing1LimitDegrees=40, swing2LimitDegrees=5, twistLimitDegrees=5}
 {
  bone1Name=spine_01
  bone2Name=pelvis
  swing1LimitDegrees=45
  swing1Motion=Free
  swing2LimitDegrees=45
  swing2Motion=Free
  twistLimitDegrees=45
  twistMotion=Free
 }
 {
  bone1Name=spine_02
  bone2Name=pelvis
  swing1LimitDegrees=45
  swing1Motion=Free
  swing2LimitDegrees=45
  swing2Motion=Free
  twistLimitDegrees=45
  twistMotion=Free
 }
 {
  bone1Name=spine_03
  bone2Name=pelvis
  swing1LimitDegrees=45
  swing1Motion=Free
  swing2LimitDegrees=45
  swing2Motion=Free
  twistLimitDegrees=45
  twistMotion=Free
 }
 {
  bone1Name=spine_04
  bone2Name=pelvis
  swing1LimitDegrees=45
  swing1Motion=Free
  swing2LimitDegrees=45
  swing2Motion=Free
  twistLimitDegrees=45
  twistMotion=Free
 }
 {
  bone1Name=spine_05
  bone2Name=pelvis
  swing1LimitDegrees=45
  swing1Motion=Free
  swing2LimitDegrees=45
  swing2Motion=Free
  twistLimitDegrees=45
  twistMotion=Free
 }
 {
  bone1Name=head
  bone2Name=pelvis
  swing1LimitDegrees=45
  swing1Motion=Free
  swing2LimitDegrees=45
  swing2Motion=Free
  twistLimitDegrees=45
  twistMotion=Free
 }
 {
  bone1Name=upperarm_l
  bone2Name=pelvis
  swing1LimitDegrees=45
  swing1Motion=Free
  swing2LimitDegrees=45
  swing2Motion=Free
  twistLimitDegrees=45
  twistMotion=Free
 }
 {
  bone1Name=lowerarm_l
  bone2Name=pelvis
  swing1LimitDegrees=45
  swing1Motion=Free
  swing2LimitDegrees=45
  swing2Motion=Free
  twistLimitDegrees=45
  twistMotion=Free
 }
 {
  bone1Name=upperarm_r
  bone2Name=pelvis
  swing1LimitDegrees=45
  swing1Motion=Free
  swing2LimitDegrees=45
  swing2Motion=Free
  twistLimitDegrees=45
  twistMotion=Free
 }
 {
  bone1Name=lowerarm_r
  bone2Name=pelvis
  swing1LimitDegrees=45
  swing1Motion=Free
  swing2LimitDegrees=45
  swing2Motion=Free
  twistLimitDegrees=45
  twistMotion=Free
 }
 {
  bone1Name=thigh_l
  bone2Name=pelvis
  swing1LimitDegrees=45
  swing1Motion=Free
  swing2LimitDegrees=45
  swing2Motion=Free
  twistLimitDegrees=45
  twistMotion=Free
 }
 {
  bone1Name=thigh_r
  bone2Name=pelvis
  swing1LimitDegrees=45
  swing1Motion=Free
  swing2LimitDegrees=45
  swing2Motion=Free
  twistLimitDegrees=45
  twistMotion=Free
 }
]
