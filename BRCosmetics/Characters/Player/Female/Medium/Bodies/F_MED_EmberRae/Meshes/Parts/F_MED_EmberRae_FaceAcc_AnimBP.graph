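# Post-style animation graph for Ember Rae's FaceAcc mesh that simulates her hair braid bone chain.
# Braid paths include RigidBody with separate skydiving and regular physics assets, AnimDynamics, and Trail controllers, with bools such as bIsSkyDiving and bAnimDynamicsEnabled choosing between branches.
# Layered constraints to spine, neck and arm bones plus additive offsets shape each result, partly weighted by crouch, jump, melee and skydive variables to limit clipping.
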AnimBlueprint /BRCosmetics/Characters/Player/Female/Medium/Bodies/F_MED_EmberRae/Meshes/Parts/F_MED_EmberRae_FaceAcc_AnimBP
target_skeleton=&/F_MED_EmberRae_FaceAcc_Skeleton
graphs [
 AnimGraph {nodes [
   n1 = Get arms up2() -> (arms up2:double)
   "Constraint - dyn_hair_braid_5 (T)"(Alpha=@n1, ConstraintWeights_0=1) {
    alphaBoolBlend {blendOption=Linear}
    boneToModify {boneName=dyn_hair_braid_5}
    constraintSetup [{transformType=Translation, targetBone {boneName=spine_01}}]
    constraintWeights [1]
   }
   n2 = Get braid_01_jittering() -> (braid_01_jittering:double)
   "Transform (Modify) Bone - Bone: dyn_hair_braid_3"(Alpha=@n2, Translation="0,1,0", Rotation="0,0,0", Scale="1,1,1") {
    rotationMode=BMM_Additive
    rotationSpace=BCS_BoneSpace
    translationMode=BMM_Additive
    translationSpace=BCS_BoneSpace
    alphaBoolBlend {blendInTime=0.2, blendOption=Linear, blendOutTime=0.2}
    boneToModify {boneName=dyn_hair_braid_3}
    translation {y=1}
   }
   n3 = "Constraint - dyn_hair_braid_2 (P)"!(ComponentPose, bAlphaBoolEnabled:bool=Get alpha crouching or jumping() -> (alpha crouching or jumping:bool), ConstraintWeights_0=1) -> (Pose) {
    alphaInputType=Bool
    alphaBoolBlend {blendInTime=0.22, blendOption=Linear, blendOutTime=0.2}
    boneToModify {boneName=dyn_hair_braid_2}
    constraintSetup [{targetBone {boneName=head}}]
    constraintWeights [1]
   }
   n4 = "Transform (Modify) Bone - Bone: dyn_hair_braid_2"!(ComponentPose=@n3, bAlphaBoolEnabled:bool=True, Translation="0,-1.5,0", Rotation="0,0,0", Scale="1,1,1") -> (Pose) {
    alphaInputType=Bool
    translationMode=BMM_Additive
    translationSpace=BCS_BoneSpace
    alphaBoolBlend {blendInTime=0.3, blendOption=Linear, blendOutTime=0.3}
    boneToModify {boneName=dyn_hair_braid_2}
    translation {y=-1.5}
   }
   n5 = "Constraint - dyn_hair_braid_3 (P)"(ComponentPose=@n4, Alpha=Get sky diving clipping alpha() -> (sky diving clipping alpha:double), ConstraintWeights_0=1) {
    alphaBoolBlend {blendInTime=0.4, blendOption=Linear, blendOutTime=0.4}
    boneToModify {boneName=dyn_hair_braid_3}
    constraintSetup [{targetBone {boneName=spine_05}}]
    constraintWeights [1]
   }
   n6 = "Transform (Modify) Bone - Bone: dyn_hair_braid_2"(ComponentPose=@n5, bAlphaBoolEnabled=True, Translation="0,0,0", Rotation="0,-13,0", Scale="1,1,1") {
    alphaInputType=Bool
    rotationMode=BMM_Additive
    rotationSpace=BCS_BoneSpace
    translationMode=BMM_Additive
    translationSpace=BCS_BoneSpace
    alphaBoolBlend {blendInTime=0.3, blendOption=Linear, blendOutTime=0.3}
    boneToModify {boneName=dyn_hair_braid_2}
    rotation {yaw=-13}
   }
   n7 = "Transform (Modify) Bone - Bone: dyn_hair_braid_2"(ComponentPose=@n6, bAlphaBoolEnabled=True, Translation="0,0,0", Rotation="0,-40,0", Scale="1,1,1") {
    alphaInputType=Bool
    rotationMode=BMM_Additive
    rotationSpace=BCS_BoneSpace
    translationMode=BMM_Additive
    translationSpace=BCS_BoneSpace
    alphaBoolBlend {blendInTime=0.2, blendOption=Linear, blendOutTime=0.4}
    boneToModify {boneName=dyn_hair_braid_2}
    rotation {yaw=-40}
   }
   "Transform (Modify) Bone - Bone: dyn_hair_braid_2"(ComponentPose=@n7, bAlphaBoolEnabled=True, Translation="0,0,0", Rotation="0,-10,0", Scale="1,1,1") {
    alphaInputType=Bool
    rotationMode=BMM_Additive
    rotationSpace=BCS_BoneSpace
    translationMode=BMM_Additive
    translationSpace=BCS_BoneSpace
    alphaBoolBlend {blendInTime=0.2, blendOption=Linear, blendOutTime=0.4}
    boneToModify {boneName=dyn_hair_braid_2}
    rotation {yaw=-10}
   }
   n8 = "Constraint - dyn_hair_braid_2 (P)"!(ComponentPose=Local To Component(LocalPose=Use cached pose 'low_perf_fringe'()), Alpha:float=0.7, ConstraintWeights_0=1) -> (Pose) {
    alphaBoolBlend {blendOption=Linear}
    boneToModify {boneName=dyn_hair_braid_2}
    constraintSetup [{targetBone {boneName=neck_01}}]
    constraintWeights [1]
   }
   n9 = "Constraint - dyn_hair_braid_3 (P)"(ComponentPose=@n8, Alpha=0.7, ConstraintWeights_0=1) {
    alphaBoolBlend {blendOption=Linear}
    boneToModify {boneName=dyn_hair_braid_3}
    constraintSetup [{targetBone {boneName=spine_05}}]
    constraintWeights [1]
   }
   n10 = "Constraint - dyn_hair_braid_4 (PP)"!(ComponentPose=@n9, Alpha=0.7, ConstraintWeights_0=0.6, ConstraintWeights_1:float=0.4) -> (Pose) {
    alphaBoolBlend {blendOption=Linear}
    boneToModify {boneName=dyn_hair_braid_4}
    constraintSetup [{targetBone {boneName=spine_04}}, {targetBone {boneName=spine_03}}]
    constraintWeights [1,1]
   }
   n11 = "Constraint - dyn_hair_braid_5 (P)"(ComponentPose=@n10, Alpha=0.7, ConstraintWeights_0=1) {
    alphaBoolBlend {blendOption=Linear}
    boneToModify {boneName=dyn_hair_braid_5}
    constraintSetup [{targetBone {boneName=spine_03}}]
    constraintWeights [1]
   }
   n12 = "Constraint - dyn_hair_braid_6 (P)"(ComponentPose=@n11, Alpha=0.5, ConstraintWeights_0=1) {
    alphaBoolBlend {blendOption=Linear}
    boneToModify {boneName=dyn_hair_braid_6}
    constraintSetup [{targetBone {boneName=spine_01}}]
    constraintWeights [1]
   }
   n13 = "Transform (Modify) Bone - Bone: dyn_hair_braid_3"(ComponentPose=@n12, Alpha=1, Translation="0,3.4,0", Rotation="0,0,0", Scale="1,0.7,1") {
    rotationMode=BMM_Additive
    rotationSpace=BCS_BoneSpace
    scaleMode=BMM_Additive
    scaleSpace=BCS_BoneSpace
    translationMode=BMM_Additive
    translationSpace=BCS_BoneSpace
    alphaBoolBlend {blendOption=Linear}
    boneToModify {boneName=dyn_hair_braid_3}
    scale {y=0.7}
    translation {y=3.4}
   }
   n14 = "Transform (Modify) Bone - Bone: dyn_hair_braid_4"(ComponentPose=@n13, Alpha=1, Translation="0,3,0", Rotation="0,0,0", Scale="1,1,1") {
    rotationMode=BMM_Additive
    rotationSpace=BCS_BoneSpace
    translationMode=BMM_Additive
    translationSpace=BCS_BoneSpace
    alphaBoolBlend {blendOption=Linear}
    boneToModify {boneName=dyn_hair_braid_4}
    translation {y=3}
   }
   n15 = "Transform (Modify) Bone - Bone: dyn_hair_braid_5"(ComponentPose=@n14, Alpha=1, Translation="0,2.5,0", Rotation="0,0,0", Scale="1,1,1") {
    rotationMode=BMM_Additive
    rotationSpace=BCS_BoneSpace
    translationMode=BMM_Additive
    translationSpace=BCS_BoneSpace
    alphaBoolBlend {blendOption=Linear}
    boneToModify {boneName=dyn_hair_braid_5}
    translation {y=2.5}
   }
   n16 = "Constraint - dyn_hair_braid_2 (P)"!(ComponentPose=@n15, bAlphaBoolEnabled:bool=True, ConstraintWeights_0=1) -> (Pose) {
    alphaInputType=Bool
    alphaBoolBlend {blendInTime=0.5, blendOption=Linear, blendOutTime=0.2}
    boneToModify {boneName=dyn_hair_braid_2}
    constraintSetup [{targetBone {boneName=neck_02}}]
    constraintWeights [1]
   }
   n17 = "Constraint - dyn_hair_braid_3 (P)"!(ComponentPose=@n16, bAlphaBoolEnabled:bool=True, ConstraintWeights_0=1) -> (Pose) {
    alphaInputType=Bool
    alphaBoolBlend {blendInTime=0.5, blendOption=Linear, blendOutTime=0.2}
    boneToModify {boneName=dyn_hair_braid_3}
    constraintSetup [{targetBone {boneName=spine_05}}]
    constraintWeights [1]
   }
   n18 = "Constraint - dyn_hair_braid_4 (P)"!(ComponentPose=@n17, bAlphaBoolEnabled:bool=True, ConstraintWeights_0=1) -> (Pose) {
    alphaInputType=Bool
    alphaBoolBlend {blendInTime=0.5, blendOption=Linear, blendOutTime=0.2}
    boneToModify {boneName=dyn_hair_braid_4}
    constraintSetup [{targetBone {boneName=spine_04}}]
    constraintWeights [1]
   }
   n19 = "Constraint - dyn_hair_braid_5 (PPP)"!(ComponentPose=@n18, bAlphaBoolEnabled:bool=True, ConstraintWeights_0=0.4, ConstraintWeights_1:float=0.4, ConstraintWeights_2:float=0.2) -> (Pose) {
    alphaInputType=Bool
    alphaBoolBlend {blendInTime=0.5, blendOption=Linear, blendOutTime=0.2}
    boneToModify {boneName=dyn_hair_braid_5}
    constraintSetup [{targetBone {boneName=spine_01}}, {targetBone {boneName=spine_02}}, {targetBone {boneName=pelvis}}]
    constraintWeights [1,1,1]
   }
   n20 = "Constraint - dyn_hair_braid_6 (PP)"!(ComponentPose=@n19, bAlphaBoolEnabled:bool=True, ConstraintWeights_0=0.5, ConstraintWeights_1:float=0.5) -> (Pose) {
    alphaInputType=Bool
    alphaBoolBlend {blendInTime=0.5, blendOption=Linear, blendOutTime=0.2}
    boneToModify {boneName=dyn_hair_braid_6}
    constraintSetup [{targetBone {boneName=pelvis}}, {targetBone {boneName=spine_01}}]
    constraintWeights [1,1]
   }
   n21 = "Transform (Modify) Bone - Bone: dyn_hair_braid_4"!(ComponentPose=@n20, bAlphaBoolEnabled:bool=True, Translation="0,3,0", Rotation="0,0,0", Scale="1,1,1") -> (Pose) {
    alphaInputType=Bool
    rotationMode=BMM_Additive
    rotationSpace=BCS_BoneSpace
    translationMode=BMM_Additive
    translationSpace=BCS_BoneSpace
    alphaBoolBlend {blendInTime=0.2, blendOption=Linear, blendOutTime=0.2}
    boneToModify {boneName=dyn_hair_braid_4}
    translation {y=3}
   }
   n22 = "Transform (Modify) Bone - Bone: dyn_hair_braid_5"!(ComponentPose=@n21, bAlphaBoolEnabled:bool=True, Translation="0,2,0", Rotation="0,0,0", Scale="1,1,1") -> (Pose) {
    alphaInputType=Bool
    rotationMode=BMM_Additive
    rotationSpace=BCS_BoneSpace
    translationMode=BMM_Additive
    translationSpace=BCS_BoneSpace
    alphaBoolBlend {blendInTime=0.2, blendOption=Linear, blendOutTime=0.2}
    boneToModify {boneName=dyn_hair_braid_5}
    translation {y=2}
   }
   n23 = "Transform (Modify) Bone - Bone: dyn_hair_braid_5"(ComponentPose=@n22, bAlphaBoolEnabled=True, Translation="0,-2,0", Rotation="0,-11,0", Scale="1,1,1") {
    alphaInputType=Bool
    rotationMode=BMM_Additive
    rotationSpace=BCS_BoneSpace
    translationMode=BMM_Additive
    translationSpace=BCS_BoneSpace
    alphaBoolBlend {blendInTime=0.2, blendOption=Linear, blendOutTime=0.2}
    boneToModify {boneName=dyn_hair_braid_5}
    rotation {yaw=-11}
    translation {y=-2}
   }
   n24 = Get braid_01_jittering()
   n25 = "Transform (Modify) Bone - Bone: dyn_hair_braid_1"(ComponentPose=@n23, Alpha=@n24, Translation="0,-1,0", Rotation="0,-2,0", Scale="1,1,1") {
    rotationMode=BMM_Additive
    rotationSpace=BCS_BoneSpace
    translationMode=BMM_Additive
    translationSpace=BCS_BoneSpace
    alphaBoolBlend {blendInTime=0.2, blendOption=Linear, blendOutTime=0.2}
    boneToModify {boneName=dyn_hair_braid_1}
    rotation {yaw=-2}
    translation {y=-1}
   }
   n26 = "Transform (Modify) Bone - Bone: dyn_hair_braid_2"!(ComponentPose=@n25, Alpha:float=@n24, Translation="0,-1,0", Rotation="0,0,0", Scale="1,1,1") -> (Pose) {
    rotationMode=BMM_Additive
    rotationSpace=BCS_BoneSpace
    translationMode=BMM_Additive
    translationSpace=BCS_BoneSpace
    alphaBoolBlend {blendInTime=0.2, blendOption=Linear, blendOutTime=0.2}
    boneToModify {boneName=dyn_hair_braid_2}
    translation {y=-1}
   }
   n27 = "Transform (Modify) Bone - Bone: dyn_hair_braid_4"!(ComponentPose=@n26, Alpha:float=@n24, Translation="0,1,0", Rotation="0,0,0", Scale="1,1,1") -> (Pose) {
    rotationMode=BMM_Additive
    rotationSpace=BCS_BoneSpace
    translationMode=BMM_Additive
    translationSpace=BCS_BoneSpace
    alphaBoolBlend {blendInTime=0.2, blendOption=Linear, blendOutTime=0.2}
    boneToModify {boneName=dyn_hair_braid_4}
    translation {y=1}
   }
   n28 = "Transform (Modify) Bone - Bone: dyn_hair_braid_5"!(ComponentPose=@n27, Alpha:float=@n24, Translation="0,1,0", Rotation="0,0,0", Scale="1,1,1") -> (Pose) {
    rotationMode=BMM_Additive
    rotationSpace=BCS_BoneSpace
    translationMode=BMM_Additive
    translationSpace=BCS_BoneSpace
    alphaBoolBlend {blendInTime=0.2, blendOption=Linear, blendOutTime=0.2}
    boneToModify {boneName=dyn_hair_braid_5}
    translation {y=1}
   }
   n29 = Trail controller - Bone: dyn_hair_braid_6!(
    ComponentPose=Local To Component(LocalPose=Use cached pose 'low_perf_fringe'())
    Alpha=1
    RelaxationSpeedScale:float=Get relaxation fps() -> (relaxation fps:double)
   ) -> (Pose) {
    bShowBaseMotion=false
    bShowLimit=false
    bUsePlanarLimit=true
    chainLength=5
    maxDeltaTime=0.3
    relaxationSpeedScale=0.6
    planarLimits [{drivingBone {boneName=spine_03}, planeTransform {location=<0 -10 0>, rotation=<rot 0 0 -90>, scale=<1 1 1>}}]
    rotationLimits [
     {limitMax=<30 30 180>, limitMin=<-30 -30 -180>}
     {limitMax=<30 30 180>, limitMin=<-30 -30 -180>}
     {limitMax=<30 30 180>, limitMin=<-30 -30 -180>}
     {limitMax=<30 30 180>, limitMin=<-30 -30 -180>}
     {limitMax=<30 30 180>, limitMin=<-30 -30 -180>}
    ]
    rotationOffsets [<0 0 0>, <0 0 0>, <0 0 0>, <0 0 0>, <0 0 0>]
    trailBone {boneName=dyn_hair_braid_6}
    trailRelaxationSpeed {editorCurveData {keys [<at 0 16 RCIM_Linear>,<at 1 14 RCIM_Linear>]}}
   }
   n30 = "Constraint - dyn_hair_braid_2 (PPP)"!(ComponentPose=@n29, Alpha=1, ConstraintWeights_0=0.6, ConstraintWeights_1:float=0.2, ConstraintWeights_2:float=0.2) -> (Pose) {
    alphaBoolBlend {blendInTime=0.3, blendOption=Linear, blendOutTime=0.3}
    boneToModify {boneName=dyn_hair_braid_2}
    constraintSetup [{targetBone {boneName=spine_05}}, {targetBone {boneName=upperarm_l}}, {targetBone {boneName=upperarm_r}}]
    constraintWeights [1,1,1]
   }
   n31 = "Constraint - dyn_hair_braid_3 (PPPP)"!(
    ComponentPose=@n30
    Alpha=1
    ConstraintWeights_0=0.6
    ConstraintWeights_1:float=0.2
    ConstraintWeights_2:float=0.2
    ConstraintWeights_3:float=0
   ) -> (Pose) {
    alphaBoolBlend {blendInTime=0.3, blendOption=Linear, blendOutTime=0.3}
    boneToModify {boneName=dyn_hair_braid_3}
    constraintSetup [{targetBone {boneName=spine_05}}, {targetBone {boneName=upperarm_l}}, {targetBone {boneName=upperarm_r}}, {targetBone {boneName=spine_04}}]
    constraintWeights [1,1,1,1]
   }
   n32 = "Constraint - dyn_hair_braid_4 (PPP)"!(ComponentPose=@n31, Alpha=0.5, ConstraintWeights_0=0.3, ConstraintWeights_1:float=0.2, ConstraintWeights_2:float=0.5) -> (Pose) {
    alphaBoolBlend {blendInTime=0.3, blendOption=Linear, blendOutTime=0.3}
    boneToModify {boneName=dyn_hair_braid_4}
    constraintSetup [{targetBone {boneName=spine_04}}, {targetBone {boneName=spine_03}}, {targetBone {boneName=dyn_hair_braid_3}}]
    constraintWeights [1,1,1]
   }
   n33 = "Constraint - dyn_hair_braid_5 (PPP)"!(ComponentPose=@n32, Alpha:float=0.6, ConstraintWeights_0=0.3, ConstraintWeights_1=0.2, ConstraintWeights_2=0.5) -> (Pose) {
    alphaBoolBlend {blendInTime=0.3, blendOption=Linear, blendOutTime=0.3}
    boneToModify {boneName=dyn_hair_braid_5}
    constraintSetup [{targetBone {boneName=spine_03}}, {targetBone {boneName=spine_02}}, {targetBone {boneName=dyn_hair_braid_4}}]
    constraintWeights [1,1,1]
   }
   n34 = "Constraint - dyn_hair_braid_6 (PPP)"!(ComponentPose=@n33, Alpha=0.8, ConstraintWeights_0=0.3, ConstraintWeights_1:float=0.2, ConstraintWeights_2:float=0.5) -> (Pose) {
    alphaBoolBlend {blendInTime=0.3, blendOption=Linear, blendOutTime=0.3}
    boneToModify {boneName=dyn_hair_braid_6}
    constraintSetup [{targetBone {boneName=spine_01}}, {targetBone {boneName=spine_02}}, {targetBone {boneName=dyn_hair_braid_5}}]
    constraintWeights [1,1,1]
   }
   n35 = "Transform (Modify) Bone - Bone: dyn_hair_braid_2"(ComponentPose=@n34, Alpha=1, Translation="0,0,0", Rotation="0,5,0", Scale="1,1,1") {
    rotationMode=BMM_Additive
    rotationSpace=BCS_BoneSpace
    translationMode=BMM_Additive
    translationSpace=BCS_BoneSpace
    alphaBoolBlend {blendOption=Linear}
    boneToModify {boneName=dyn_hair_braid_2}
    rotation {yaw=5}
   }
   n36 = "Transform (Modify) Bone - Bone: dyn_hair_braid_2"(ComponentPose=@n35, Alpha=1, Translation="0,0,0", Rotation="0,-5,0", Scale="1,1,1") {
    rotationMode=BMM_Additive
    rotationSpace=BCS_BoneSpace
    translationMode=BMM_Additive
    translationSpace=BCS_BoneSpace
    alphaBoolBlend {blendOption=Linear}
    boneToModify {boneName=dyn_hair_braid_2}
    rotation {yaw=-5}
   }
   n37 = Get braid trail alpha() -> (braid trail alpha:double)
   n38 = "Constraint - dyn_hair_braid_2 (PPP)"(ComponentPose=@n36, Alpha=@n37, ConstraintWeights_0=0.6, ConstraintWeights_1=0.2, ConstraintWeights_2=0.2) {
    alphaBoolBlend {blendInTime=0.3, blendOption=Linear, blendOutTime=0.3}
    boneToModify {boneName=dyn_hair_braid_2}
    constraintSetup [{targetBone {boneName=spine_05}}, {targetBone {boneName=upperarm_l}}, {targetBone {boneName=upperarm_r}}]
    constraintWeights [1,1,1]
   }
   n39 = "Constraint - dyn_hair_braid_3 (PPP)"!(ComponentPose=@n38, Alpha=@n37, ConstraintWeights_0=0.6, ConstraintWeights_1:float=0.2, ConstraintWeights_2:float=0.2) -> (Pose) {
    alphaBoolBlend {blendInTime=0.3, blendOption=Linear, blendOutTime=0.3}
    boneToModify {boneName=dyn_hair_braid_3}
    constraintSetup [{targetBone {boneName=spine_05}}, {targetBone {boneName=upperarm_l}}, {targetBone {boneName=upperarm_r}}]
    constraintWeights [1,1,1]
   }
   n40 = "Constraint - dyn_hair_braid_4 (PPP)"(ComponentPose=@n39, Alpha=@n37, ConstraintWeights_0=0.5, ConstraintWeights_1=0.3, ConstraintWeights_2=0.2) {
    alphaBoolBlend {blendInTime=0.3, blendOption=Linear, blendOutTime=0.3}
    boneToModify {boneName=dyn_hair_braid_4}
    constraintSetup [{targetBone {boneName=spine_04}}, {targetBone {boneName=spine_03}}, {targetBone {boneName=dyn_hair_braid_3}}]
    constraintWeights [1,1,1]
   }
   n41 = "Constraint - dyn_hair_braid_5 (PPP)"(ComponentPose=@n40, Alpha=@n37, ConstraintWeights_0=0.2, ConstraintWeights_1=0.6, ConstraintWeights_2=0.2) {
    alphaBoolBlend {blendInTime=0.3, blendOption=Linear, blendOutTime=0.3}
    boneToModify {boneName=dyn_hair_braid_5}
    constraintSetup [{targetBone {boneName=spine_03}}, {targetBone {boneName=spine_02}}, {targetBone {boneName=dyn_hair_braid_4}}]
    constraintWeights [1,1,1]
   }
   n42 = "Constraint - dyn_hair_braid_6 (PPP)"(ComponentPose=@n41, Alpha=@n37, ConstraintWeights_0=0.4, ConstraintWeights_1=0.3, ConstraintWeights_2=0.3) {
    alphaBoolBlend {blendInTime=0.3, blendOption=Linear, blendOutTime=0.3}
    boneToModify {boneName=dyn_hair_braid_6}
    constraintSetup [{targetBone {boneName=spine_01}}, {targetBone {boneName=spine_02}}, {targetBone {boneName=dyn_hair_braid_5}}]
    constraintWeights [1,1,1]
   }
   n43 = "Transform (Modify) Bone - Bone: dyn_hair_braid_5"(
    ComponentPose=@n42
    Alpha=Get crouching and moving 2d() -> (crouching and moving 2d:double)
    Translation="0,0,0"
    Rotation="0,5,0"
    Scale="1,1,1"
   ) {
    rotationMode=BMM_Additive
    rotationSpace=BCS_BoneSpace
    alphaBoolBlend {blendInTime=0.2, blendOption=Linear, blendOutTime=0.2}
    boneToModify {boneName=dyn_hair_braid_5}
    rotation {yaw=5}
   }
   n44 = Get braid clipping crouching trail() -> (braid clipping crouching trail:float)
   n45 = "Transform (Modify) Bone - Bone: dyn_hair_braid_3"(ComponentPose=@n43, Alpha=@n44, Translation="0,-1,0", Rotation="0,-15,0", Scale="1,1,1") {
    rotationMode=BMM_Additive
    rotationSpace=BCS_BoneSpace
    translationMode=BMM_Additive
    translationSpace=BCS_BoneSpace
    alphaBoolBlend {blendOption=Linear}
    boneToModify {boneName=dyn_hair_braid_3}
    rotation {yaw=-15}
    translation {y=-1}
   }
   n46 = "Transform (Modify) Bone - Bone: dyn_hair_braid_4"(ComponentPose=@n45, Alpha=@n44, Translation="3,0,0", Rotation="0,3,0", Scale="1,1,1") {
    rotationMode=BMM_Additive
    rotationSpace=BCS_BoneSpace
    translationMode=BMM_Additive
    translationSpace=BCS_BoneSpace
    alphaBoolBlend {blendOption=Linear}
    boneToModify {boneName=dyn_hair_braid_4}
    rotation {yaw=3}
    translation {x=3}
   }
   n47 = "Transform (Modify) Bone - Bone: dyn_hair_braid_5"(ComponentPose=@n46, Alpha=@n44, Translation="0,2,0", Rotation="0,3,0", Scale="1,1,1") {
    rotationMode=BMM_Additive
    rotationSpace=BCS_BoneSpace
    translationMode=BMM_Additive
    translationSpace=BCS_BoneSpace
    alphaBoolBlend {blendOption=Linear}
    boneToModify {boneName=dyn_hair_braid_5}
    rotation {yaw=3}
    translation {y=2}
   }
   n48 = "Transform (Modify) Bone - Bone: dyn_hair_braid_2"!(ComponentPose=@n47, bAlphaBoolEnabled:bool=True, Translation="0,0,0", Rotation="0,-20,0", Scale="1,1,1") -> (Pose) {
    alphaInputType=Bool
    rotationMode=BMM_Additive
    rotationSpace=BCS_BoneSpace
    translationMode=BMM_Additive
    translationSpace=BCS_BoneSpace
    alphaBoolBlend {blendInTime=0.1, blendOption=Linear, blendOutTime=2}
    boneToModify {boneName=dyn_hair_braid_2}
    rotation {yaw=-20}
   }
   n49 = "Constraint - dyn_hair_braid_2 (P)"(ComponentPose=@n48, bAlphaBoolEnabled=True, ConstraintWeights_0=1) {
    alphaInputType=Bool
    alphaBoolBlend {blendInTime=0.1, blendOption=Linear, blendOutTime=2}
    boneToModify {boneName=dyn_hair_braid_2}
    constraintSetup [{targetBone {boneName=spine_05}}]
    constraintWeights [1]
   }
   n50 = "Constraint - dyn_hair_braid_3 (P)"(ComponentPose=@n49, bAlphaBoolEnabled=True, ConstraintWeights_0=1) {
    alphaInputType=Bool
    alphaBoolBlend {blendInTime=0.1, blendOption=Linear, blendOutTime=2}
    boneToModify {boneName=dyn_hair_braid_3}
    constraintSetup [{targetBone {boneName=spine_04}}]
    constraintWeights [1]
   }
   n51 = "Constraint - dyn_hair_braid_4 (P)"(ComponentPose=@n50, bAlphaBoolEnabled=True, ConstraintWeights_0=1) {
    alphaInputType=Bool
    alphaBoolBlend {blendInTime=0.1, blendOption=Linear, blendOutTime=2}
    boneToModify {boneName=dyn_hair_braid_4}
    constraintSetup [{targetBone {boneName=spine_03}}]
    constraintWeights [1]
   }
   n52 = "Constraint - dyn_hair_braid_5 (P)"!(ComponentPose=@n51, bAlphaBoolEnabled:bool=True, ConstraintWeights_0=1) -> (Pose) {
    alphaInputType=Bool
    alphaBoolBlend {blendInTime=0.1, blendOption=Linear, blendOutTime=2}
    boneToModify {boneName=dyn_hair_braid_5}
    constraintSetup [{targetBone {boneName=spine_02}}]
    constraintWeights [1]
   }
   n53 = "Constraint - dyn_hair_braid_6 (P)"!(ComponentPose=@n52, bAlphaBoolEnabled:bool=True, ConstraintWeights_0=1) -> (Pose) {
    alphaInputType=Bool
    alphaBoolBlend {blendInTime=0.1, blendOption=Linear, blendOutTime=2}
    boneToModify {boneName=dyn_hair_braid_6}
    constraintSetup [{targetBone {boneName=spine_01}}]
    constraintWeights [1]
   }
   n54 = "Transform (Modify) Bone - Bone: dyn_hair_braid_6"!(ComponentPose=@n53, bAlphaBoolEnabled:bool=True, Translation="0,0,0", Rotation="0,-6,0", Scale="1,1,1") -> (Pose) {
    alphaInputType=Bool
    rotationMode=BMM_Additive
    rotationSpace=BCS_BoneSpace
    alphaBoolBlend {blendInTime=0.8, blendOption=Linear, blendOutTime=0.4}
    boneToModify {boneName=dyn_hair_braid_6}
    rotation {yaw=-6}
   }
   n55 = "Transform (Modify) Bone - Bone: dyn_hair_braid_5"!(ComponentPose=@n54, bAlphaBoolEnabled:bool=True, Translation="0,0,0", Rotation="0,-15,0", Scale="1,1,1") -> (Pose) {
    alphaInputType=Bool
    rotationMode=BMM_Additive
    rotationSpace=BCS_BoneSpace
    alphaBoolBlend {blendInTime=0.8, blendOption=Linear, blendOutTime=0.4}
    boneToModify {boneName=dyn_hair_braid_5}
    rotation {yaw=-15}
   }
   n56 = "Transform (Modify) Bone - Bone: dyn_hair_braid_4"!(ComponentPose=@n55, bAlphaBoolEnabled:bool=True, Translation="0,0,0", Rotation="0,-10,0", Scale="1,1,1") -> (Pose) {
    alphaInputType=Bool
    rotationMode=BMM_Additive
    rotationSpace=BCS_BoneSpace
    alphaBoolBlend {blendInTime=0.8, blendOption=Linear, blendOutTime=0.4}
    boneToModify {boneName=dyn_hair_braid_4}
    rotation {yaw=-10}
   }
   n57 = "Transform (Modify) Bone - Bone: dyn_hair_braid_3"!(ComponentPose=@n56, bAlphaBoolEnabled:bool=True, Translation="0,0,0", Rotation="0,-8,0", Scale="1,1,1") -> (Pose) {
    alphaInputType=Bool
    rotationMode=BMM_Additive
    rotationSpace=BCS_BoneSpace
    alphaBoolBlend {blendInTime=0.8, blendOption=Linear, blendOutTime=0.4}
    boneToModify {boneName=dyn_hair_braid_3}
    rotation {yaw=-8}
   }
   n58 = "Transform (Modify) Bone - Bone: dyn_hair_braid_2"(ComponentPose=@n57, bAlphaBoolEnabled=True, Translation="0,0,0", Rotation="0,-4,0", Scale="1,1,1") {
    alphaInputType=Bool
    rotationMode=BMM_Additive
    rotationSpace=BCS_BoneSpace
    alphaBoolBlend {blendInTime=0.8, blendOption=Linear, blendOutTime=0.4}
    boneToModify {boneName=dyn_hair_braid_2}
    rotation {yaw=-4}
   }
   n59 = Component To Local(ComponentPose=@n58)
   n60 = Trail controller - Bone: dyn_hair_braid_6!(
    ComponentPose=Local To Component(LocalPose=Use cached pose 'low_perf_fringe'())
    bAlphaBoolEnabled:bool=Get jumping_cape_bone() -> (jumping_cape_bone:bool)
    RelaxationSpeedScale=1.5
   ) -> (Pose) {
    alphaInputType=Bool
    bShowBaseMotion=false
    bShowLimit=false
    bUsePlanarLimit=true
    chainLength=5
    maxDeltaTime=0.3
    relaxationSpeedScale=0.6
    planarLimits [{drivingBone {boneName=spine_01}, planeTransform {location=<0 -10 0>, rotation=<rot 0 0 -90>, scale=<1 1 1>}}]
    rotationLimits [
     {limitMax=<30 30 180>, limitMin=<-30 -30 -180>}
     {limitMax=<30 30 180>, limitMin=<-30 -30 -180>}
     {limitMax=<30 30 180>, limitMin=<-30 -30 -180>}
     {limitMax=<30 30 180>, limitMin=<-30 -30 -180>}
     {limitMax=<30 30 180>, limitMin=<-30 -30 -180>}
    ]
    rotationOffsets [<0 0 0>, <0 0 0>, <0 0 0>, <0 0 0>, <0 0 0>]
    trailBone {boneName=dyn_hair_braid_6}
    trailRelaxationSpeed {editorCurveData {keys [<at 0 25 RCIM_Linear>,<at 1 12 RCIM_Linear>]}}
   }
   n61 = "Anim Dynamics - Chain: dyn_hair_braid_2 -> dyn_hair_braid_6"!(ComponentPose=@n60, Alpha=1, GravityOverride:Vector=Get Gravity Override(), ExternalForce="0,0,0") -> (Pose) {
    angularBiasOverride=0.9
    angularDampingOverride=0.95
    angularSpringConstant=100
    bAngularSpring=true
    bChain=true
    bOverrideAngularDamping=true
    bUseGravityOverride=true
    bUsePlanarLimit=false
    bUseSphericalLimits=true
    gravityScale=0.016
    alphaBoolBlend {blendOption=Linear}
    boundBone {boneName=dyn_hair_braid_2}
    chainEnd {boneName=dyn_hair_braid_6}
    componentAppliedLinearAccClamp=<50 50 50>
    componentLinearAccScale=<0.1 0.1 0.1>
    componentLinearVelScale=<0.1 0.1 0.1>
    gravityOverride {z=-980}
    physicsBodyDefinitions [
     {
      boundBone {boneName=dyn_hair_braid_2}
      boxExtents=<40 40 40>
      constraintSetup {angularLimitsMax=<0 20 20>, angularLimitsMin=<0 -20 -20>, angularTarget=<1 0 0>}
      localJointOffset=<0 0 0>
     }
     {
      boundBone {boneName=dyn_hair_braid_3}
      boxExtents=<38 38 38>
      constraintSetup {angularLimitsMax=<30 30 30>, angularLimitsMin=<-30 -30 -30>, angularTarget=<1 0 0>}
      localJointOffset=<0 0 0>
     }
     {
      boundBone {boneName=dyn_hair_braid_4}
      boxExtents=<35 35 35>
      constraintSetup {angularLimitsMax=<20 20 20>, angularLimitsMin=<-20 -20 -20>, angularTarget=<1 0 0>}
      localJointOffset=<0 0 0>
     }
     {
      boundBone {boneName=dyn_hair_braid_5}
      boxExtents=<32 32 32>
      constraintSetup {angularLimitsMax=<20 20 20>, angularLimitsMin=<-20 -20 -20>, angularTarget=<1 0 0>}
      localJointOffset=<0 0 0>
     }
     {
      boundBone {boneName=dyn_hair_braid_6}
      boxExtents=<30 30 30>
      constraintSetup {angularLimitsMax=<20 20 15>, angularLimitsMin=<-20 -20 -15>, angularTarget=<1 0 0>}
      localJointOffset=<0 0 0>
     }
    ]
    planarLimits [
     {drivingBone {boneName=pelvis}, planeTransform {location=<0.2877 -35.9728 -40.2381>, rotation=<rot 10.0853 9.2932 -32.0367>}}
     {drivingBone {boneName=pelvis}, planeTransform {location=<-2.0078 -27.1051 53.9388>, rotation=<rot -10.1745 7.5795 -131.8613>}}
    ]
    relativeSpaceBone {boneName=spine_01}
    sphericalLimits [
     {limitRadius=20, drivingBone {boneName=spine_05}, sphereLocalOffset=<0 8 0>}
     {limitRadius=18, drivingBone {boneName=spine_04}, sphereLocalOffset=<0 8 0>}
     {limitRadius=20, drivingBone {boneName=spine_02}, sphereLocalOffset=<0 8 0>}
     {limitRadius=28, drivingBone {boneName=pelvis}, sphereLocalOffset=<-2 12 -2>}
     {limitRadius=14, drivingBone {boneName=upperarm_l}, sphereLocalOffset=<1 7 0>}
     {limitRadius=10, drivingBone {boneName=upperarm_r}, sphereLocalOffset=<-3 -5 0>}
     {limitRadius=13, drivingBone {boneName=upperarm_twist_01_l}, sphereLocalOffset=<12 0 0>}
     {limitRadius=12, drivingBone {boneName=upperarm_r}, sphereLocalOffset=<-18 -3 0>}
     {limitRadius=16, drivingBone {boneName=thigh_l}, sphereLocalOffset=<4 0 -7>}
     {limitRadius=17, drivingBone {boneName=thigh_r}, sphereLocalOffset=<-4 3 7>}
    ]
   }
   n62 = Get melee or crouching dynamics() -> (melee or crouching dynamics:double)
   n63 = "Constraint - dyn_hair_braid_2 (PPP)"(ComponentPose=@n61, Alpha=@n62, ConstraintWeights_0=0.6, ConstraintWeights_1=0.2, ConstraintWeights_2=0.2) {
    alphaBoolBlend {blendOption=Linear}
    boneToModify {boneName=dyn_hair_braid_2}
    constraintSetup [{targetBone {boneName=spine_05}}, {targetBone {boneName=upperarm_l}}, {targetBone {boneName=upperarm_r}}]
    constraintWeights [1,1,1]
   }
   n64 = "Constraint - dyn_hair_braid_3 (PPP)"(ComponentPose=@n63, Alpha=@n62, ConstraintWeights_0=0.6, ConstraintWeights_1=0.2, ConstraintWeights_2=0.2) {
    alphaBoolBlend {blendOption=Linear}
    boneToModify {boneName=dyn_hair_braid_3}
    constraintSetup [{targetBone {boneName=spine_05}}, {targetBone {boneName=upperarm_l}}, {targetBone {boneName=upperarm_r}}]
    constraintWeights [1,1,1]
   }
   n65 = "Constraint - dyn_hair_braid_4 (PPP)"(ComponentPose=@n64, Alpha=@n62, ConstraintWeights_0=0.5, ConstraintWeights_1=0.3, ConstraintWeights_2=0.2) {
    alphaBoolBlend {blendOption=Linear}
    boneToModify {boneName=dyn_hair_braid_4}
    constraintSetup [{targetBone {boneName=spine_04}}, {targetBone {boneName=spine_03}}, {targetBone {boneName=dyn_hair_braid_3}}]
    constraintWeights [1,1,1]
   }
   n66 = "Constraint - dyn_hair_braid_5 (PPP)"(ComponentPose=@n65, Alpha=@n62, ConstraintWeights_0=0.2, ConstraintWeights_1=0.6, ConstraintWeights_2=0.2) {
    alphaBoolBlend {blendOption=Linear}
    boneToModify {boneName=dyn_hair_braid_5}
    constraintSetup [{targetBone {boneName=spine_03}}, {targetBone {boneName=spine_02}}, {targetBone {boneName=dyn_hair_braid_4}}]
    constraintWeights [1,1,1]
   }
   n67 = "Constraint - dyn_hair_braid_6 (PPPP)"!(
    ComponentPose=@n66
    Alpha=@n62
    ConstraintWeights_0=0.4
    ConstraintWeights_1:float=0.4
    ConstraintWeights_2:float=0.2
    ConstraintWeights_3:float=0
   ) -> (Pose) {
    alphaBoolBlend {blendOption=Linear}
    boneToModify {boneName=dyn_hair_braid_6}
    constraintSetup [{targetBone {boneName=spine_01}}, {targetBone {boneName=spine_02}}, {targetBone {boneName=dyn_hair_braid_5}}, {targetBone {boneName=pelvis}}]
    constraintWeights [1,1,1,1]
   }
   n68 = "Transform (Modify) Bone - Bone: dyn_hair_braid_4"!(ComponentPose=@n67, Alpha:float=@n62, Translation="0,-3,0", Rotation="0,0,0", Scale="1,1,1") -> (Pose) {
    translationMode=BMM_Additive
    translationSpace=BCS_BoneSpace
    alphaBoolBlend {blendOption=Linear}
    boneToModify {boneName=dyn_hair_braid_4}
    translation {y=-3}
   }
   n69 = "Transform (Modify) Bone - Bone: dyn_hair_braid_5"(ComponentPose=@n68, bAlphaBoolEnabled=True, Translation="0,0,0", Rotation="0,5,0", Scale="1,1,1") {
    alphaInputType=Bool
    rotationMode=BMM_Additive
    rotationSpace=BCS_BoneSpace
    translationMode=BMM_Additive
    translationSpace=BCS_BoneSpace
    alphaBoolBlend {blendInTime=0.2, blendOption=Linear, blendOutTime=0.2}
    boneToModify {boneName=dyn_hair_braid_5}
    rotation {yaw=5}
   }
   n70 = "Transform (Modify) Bone - Bone: dyn_hair_braid_4"!(ComponentPose=@n69, bAlphaBoolEnabled:bool=True, Translation="0,0,0", Rotation="0,5,0", Scale="1,1,1") -> (Pose) {
    alphaInputType=Bool
    rotationMode=BMM_Additive
    rotationSpace=BCS_BoneSpace
    translationMode=BMM_Additive
    translationSpace=BCS_BoneSpace
    alphaBoolBlend {blendInTime=0.2, blendOption=Linear, blendOutTime=0.2}
    boneToModify {boneName=dyn_hair_braid_4}
    rotation {yaw=5}
   }
   n71 = "Transform (Modify) Bone - Bone: dyn_hair_braid_5"!(ComponentPose=@n70, Alpha:float=1, Translation="0,0,0", Rotation="0,-4,0", Scale="1,1,1") -> (Pose) {
    rotationMode=BMM_Additive
    rotationSpace=BCS_BoneSpace
    translationMode=BMM_Additive
    translationSpace=BCS_BoneSpace
    alphaBoolBlend {blendInTime=0.2, blendOption=Linear, blendOutTime=0.2}
    boneToModify {boneName=dyn_hair_braid_5}
    rotation {yaw=-4}
   }
   n72 = "Transform (Modify) Bone - Bone: dyn_hair_braid_6"!(ComponentPose=@n71, Alpha:float=1, Translation="0,0,0", Rotation="0,-4,0", Scale="1,1,1") -> (Pose) {
    rotationMode=BMM_Additive
    rotationSpace=BCS_BoneSpace
    translationMode=BMM_Additive
    translationSpace=BCS_BoneSpace
    alphaBoolBlend {blendInTime=0.2, blendOption=Linear, blendOutTime=0.2}
    boneToModify {boneName=dyn_hair_braid_6}
    rotation {yaw=-4}
   }
   n73 = Component To Local(ComponentPose=@n72)
   n74 = Get bAnimDynamicsEnabled() -> (bAnimDynamicsEnabled:bool)
   n75 = "Transform (Modify) Bone - Bone: dyn_hair_braid_1"(
    ComponentPose=Local To Component(LocalPose=Blend Poses by bool(BlendPose_0=@n73, BlendPose_1=@n59, bActiveValue=@n74) {blendType=Linear})
    Alpha=Get rb off braid lifting() -> (rb off braid lifting:double)
    Translation="0,0,0"
    Rotation="0,-6,0"
    Scale="1,1,1"
   ) {
    rotationMode=BMM_Additive
    rotationSpace=BCS_BoneSpace
    translationMode=BMM_Additive
    translationSpace=BCS_BoneSpace
    alphaBoolBlend {blendOption=Linear}
    boneToModify {boneName=dyn_hair_braid_1}
    rotation {yaw=-6}
   }
   n76 = RigidBody!(
    ComponentPose=Local To Component(LocalPose=Use cached pose 'braid_simulation'())
    Alpha
    OverrideWorldGravity:Vector=Get Gravity Override()
    ExternalForce
   ) -> (Pose) {
    bClampLinearTranslationLimitToRefPose=true
    bOverrideWorldGravity=true
    overridePhysicsAsset=&/F_MED_EmberRae_FaceAcc_Skydive_Physics_braid
    simulationSpace=BaseBoneSpace
    alphaBoolBlend {blendOption=Linear}
    baseBoneRef {boneName=root}
    componentAppliedLinearAccClamp=<5000 5000 5000>
    simSpaceSettings {velocityScaleZ=0}
   }
   n77 = Component To Local(ComponentPose=@n76)
   n78 = RigidBody(ComponentPose=Local To Component(LocalPose=Use cached pose 'braid_simulation'()), OverrideWorldGravity=Get Gravity Override()) {
    bClampLinearTranslationLimitToRefPose=true
    bOverrideWorldGravity=true
    overridePhysicsAsset=&/F_MED_EmberRae_FaceAcc_Physics_braid
    simulationSpace=BaseBoneSpace
    alphaBoolBlend {blendOption=Linear}
    baseBoneRef {boneName=root}
    componentAppliedLinearAccClamp=<1800 3000 6000>
    componentLinearAccScale=<0.005 0.005 0.41>
    componentLinearVelScale=<0.005 0.005 0.9>
   }
   n79 = Component To Local(ComponentPose=@n78)
   n80 = Get bIsSkyDiving()
   n81 = Component To Local(ComponentPose=Local To Component(LocalPose=Blend Poses by bool(BlendPose_0=@n77, BlendPose_1=@n79, bActiveValue=@n80) {blendType=Linear}))
   n82 = "Constraint - dyn_hair_braid_2 (P)"!(ComponentPose=Local To Component(LocalPose=Use cached pose 'braid_simulation'()), Alpha:float=1, ConstraintWeights_0=1) -> (Pose) {
    alphaBoolBlend {blendOption=Linear}
    boneToModify {boneName=dyn_hair_braid_2}
    constraintSetup [{targetBone {boneName=neck_01}}]
    constraintWeights [1]
   }
   n83 = "Constraint - dyn_hair_braid_3 (P)"!(ComponentPose=@n82, Alpha:float=1, ConstraintWeights_0=1) -> (Pose) {
    alphaBoolBlend {blendOption=Linear}
    boneToModify {boneName=dyn_hair_braid_3}
    constraintSetup [{targetBone {boneName=spine_05}}]
    constraintWeights [1]
   }
   n84 = "Constraint - dyn_hair_braid_4 (PP)"(ComponentPose=@n83, Alpha=1, ConstraintWeights_0=0.6, ConstraintWeights_1=0.4) {
    alphaBoolBlend {blendOption=Linear}
    boneToModify {boneName=dyn_hair_braid_4}
    constraintSetup [{targetBone {boneName=spine_04}}, {targetBone {boneName=spine_01}}]
    constraintWeights [1,1]
   }
   n85 = "Constraint - dyn_hair_braid_5 (P)"!(ComponentPose=@n84, Alpha:float=1, ConstraintWeights_0=1) -> (Pose) {
    alphaBoolBlend {blendOption=Linear}
    boneToModify {boneName=dyn_hair_braid_5}
    constraintSetup [{targetBone {boneName=spine_01}}]
    constraintWeights [1]
   }
   n86 = "Constraint - dyn_hair_braid_6 (P)"!(ComponentPose=@n85, Alpha:float=1, ConstraintWeights_0=1) -> (Pose) {
    alphaBoolBlend {blendOption=Linear}
    boneToModify {boneName=dyn_hair_braid_6}
    constraintSetup [{targetBone {boneName=spine_01}}]
    constraintWeights [1]
   }
   n87 = "Transform (Modify) Bone - Bone: dyn_hair_braid_3"!(ComponentPose=@n86, Alpha:float=1, Translation="0,3.4,0", Rotation="0,0,0", Scale="1,0.7,1") -> (Pose) {
    rotationMode=BMM_Additive
    rotationSpace=BCS_BoneSpace
    scaleMode=BMM_Additive
    scaleSpace=BCS_BoneSpace
    translationMode=BMM_Additive
    translationSpace=BCS_BoneSpace
    alphaBoolBlend {blendOption=Linear}
    boneToModify {boneName=dyn_hair_braid_3}
    scale {y=0.7}
    translation {y=3.4}
   }
   n88 = "Transform (Modify) Bone - Bone: dyn_hair_braid_4"!(ComponentPose=@n87, Alpha:float=1, Translation="0,3,0", Rotation="0,0,0", Scale="1,1,1") -> (Pose) {
    rotationMode=BMM_Additive
    rotationSpace=BCS_BoneSpace
    scaleMode=BMM_Replace
    scaleSpace=BCS_BoneSpace
    translationMode=BMM_Additive
    translationSpace=BCS_BoneSpace
    alphaBoolBlend {blendOption=Linear}
    boneToModify {boneName=dyn_hair_braid_4}
    translation {y=3}
   }
   n89 = "Transform (Modify) Bone - Bone: dyn_hair_braid_5"(ComponentPose=@n88, Alpha=1, Translation="0,0,0", Rotation="0,0,0", Scale="1,1,1") {
    rotationMode=BMM_Additive
    rotationSpace=BCS_BoneSpace
    scaleMode=BMM_Replace
    scaleSpace=BCS_BoneSpace
    translationMode=BMM_Additive
    translationSpace=BCS_BoneSpace
    alphaBoolBlend {blendOption=Linear}
    boneToModify {boneName=dyn_hair_braid_5}
   }
   n90 = "Transform (Modify) Bone - Bone: dyn_hair_braid_6"(ComponentPose=@n89, Alpha=1, Translation="0,0,0", Rotation="0,-10,0", Scale="1,1,1") {
    rotationMode=BMM_Additive
    rotationSpace=BCS_BoneSpace
    scaleMode=BMM_Replace
    scaleSpace=BCS_BoneSpace
    translationMode=BMM_Additive
    translationSpace=BCS_BoneSpace
    alphaBoolBlend {blendOption=Linear}
    boneToModify {boneName=dyn_hair_braid_6}
    rotation {yaw=-10}
   }
   n91 = RigidBody(ComponentPose=@n90, Alpha=Get rb alpha() -> (rb alpha:double), OverrideWorldGravity=Get Gravity Override()) {
    bClampLinearTranslationLimitToRefPose=true
    bOverrideWorldGravity=true
    overridePhysicsAsset=&F_MED_EmberRae_FaceAcc_Physics_braid
    simulationSpace=BaseBoneSpace
    alphaBoolBlend {blendOption=Linear}
    baseBoneRef {boneName=root}
    componentAppliedLinearAccClamp=<3000 3000 3000>
    componentLinearAccScale=<0.1 0.1 0.1>
    componentLinearVelScale=<0.1 0.1 0.1>
   }
   n92 = "Constraint - dyn_hair_braid_2 (P)"!(ComponentPose=@n91, bAlphaBoolEnabled:bool=True, ConstraintWeights_0=1) -> (Pose) {
    alphaInputType=Bool
    alphaBoolBlend {blendInTime=0.5, blendOption=Linear, blendOutTime=0.2}
    boneToModify {boneName=dyn_hair_braid_2}
    constraintSetup [{targetBone {boneName=neck_02}}]
    constraintWeights [1]
   }
   n93 = "Constraint - dyn_hair_braid_3 (P)"!(ComponentPose=@n92, bAlphaBoolEnabled:bool=True, ConstraintWeights_0=1) -> (Pose) {
    alphaInputType=Bool
    alphaBoolBlend {blendInTime=0.5, blendOption=Linear, blendOutTime=0.2}
    boneToModify {boneName=dyn_hair_braid_3}
    constraintSetup [{targetBone {boneName=spine_05}}]
    constraintWeights [1]
   }
   n94 = "Constraint - dyn_hair_braid_4 (P)"(ComponentPose=@n93, bAlphaBoolEnabled=True, ConstraintWeights_0=1) {
    alphaInputType=Bool
    alphaBoolBlend {blendInTime=0.5, blendOption=Linear, blendOutTime=0.2}
    boneToModify {boneName=dyn_hair_braid_4}
    constraintSetup [{targetBone {boneName=spine_04}}]
    constraintWeights [1]
   }
   n95 = "Constraint - dyn_hair_braid_5 (PPP)"!(ComponentPose=@n94, bAlphaBoolEnabled:bool=True, ConstraintWeights_0=0.4, ConstraintWeights_1=0.4, ConstraintWeights_2=0.2) -> (Pose) {
    alphaInputType=Bool
    alphaBoolBlend {blendInTime=0.5, blendOption=Linear, blendOutTime=0.2}
    boneToModify {boneName=dyn_hair_braid_5}
    constraintSetup [{targetBone {boneName=spine_01}}, {targetBone {boneName=spine_02}}, {targetBone {boneName=pelvis}}]
    constraintWeights [1,1,1]
   }
   n96 = "Constraint - dyn_hair_braid_6 (PP)"(ComponentPose=@n95, bAlphaBoolEnabled=True, ConstraintWeights_0=0.6, ConstraintWeights_1=0.4) {
    alphaInputType=Bool
    alphaBoolBlend {blendInTime=0.5, blendOption=Linear, blendOutTime=0.2}
    boneToModify {boneName=dyn_hair_braid_6}
    constraintSetup [{targetBone {boneName=pelvis}}, {targetBone {boneName=spine_01}}]
    constraintWeights [1,1]
   }
   n97 = "Transform (Modify) Bone - Bone: dyn_hair_braid_4"!(ComponentPose=@n96, bAlphaBoolEnabled:bool=True, Translation="0,3,0", Rotation="0,0,0", Scale="1,1,1") -> (Pose) {
    alphaInputType=Bool
    rotationMode=BMM_Additive
    rotationSpace=BCS_BoneSpace
    translationMode=BMM_Additive
    translationSpace=BCS_BoneSpace
    alphaBoolBlend {blendInTime=0.2, blendOption=Linear, blendOutTime=0.2}
    boneToModify {boneName=dyn_hair_braid_4}
    translation {y=3}
   }
   n98 = "Transform (Modify) Bone - Bone: dyn_hair_braid_5"!(ComponentPose=@n97, bAlphaBoolEnabled:bool=True, Translation="0,2,0", Rotation="0,0,0", Scale="1,1,1") -> (Pose) {
    alphaInputType=Bool
    rotationMode=BMM_Additive
    rotationSpace=BCS_BoneSpace
    translationMode=BMM_Additive
    translationSpace=BCS_BoneSpace
    alphaBoolBlend {blendInTime=0.2, blendOption=Linear, blendOutTime=0.2}
    boneToModify {boneName=dyn_hair_braid_5}
    translation {y=2}
   }
   n99 = "Transform (Modify) Bone - Bone: dyn_hair_braid_5"(ComponentPose=@n98, bAlphaBoolEnabled=True, Translation="0,-2,0", Rotation="0,-11,0", Scale="1,1,1") {
    alphaInputType=Bool
    rotationMode=BMM_Additive
    rotationSpace=BCS_BoneSpace
    translationMode=BMM_Additive
    translationSpace=BCS_BoneSpace
    alphaBoolBlend {blendInTime=0.2, blendOption=Linear, blendOutTime=0.2}
    boneToModify {boneName=dyn_hair_braid_5}
    rotation {yaw=-11}
    translation {y=-2}
   }
   n100 = "Transform (Modify) Bone - Bone: dyn_hair_braid_1"(ComponentPose=@n99, Alpha=@n2, Translation="0,-1,0", Rotation="0,-2,0", Scale="1,1,1") {
    rotationMode=BMM_Additive
    rotationSpace=BCS_BoneSpace
    translationMode=BMM_Additive
    translationSpace=BCS_BoneSpace
    alphaBoolBlend {blendInTime=0.2, blendOption=Linear, blendOutTime=0.2}
    boneToModify {boneName=dyn_hair_braid_1}
    rotation {yaw=-2}
    translation {y=-1}
   }
   n101 = "Transform (Modify) Bone - Bone: dyn_hair_braid_4"!(ComponentPose=@n100, Alpha:float=@n2, Translation="0,3,0", Rotation="0,0,0", Scale="1,1,1") -> (Pose) {
    rotationMode=BMM_Additive
    rotationSpace=BCS_BoneSpace
    translationMode=BMM_Additive
    translationSpace=BCS_BoneSpace
    alphaBoolBlend {blendInTime=0.2, blendOption=Linear, blendOutTime=0.2}
    boneToModify {boneName=dyn_hair_braid_4}
    translation {y=3}
   }
   n102 = "Transform (Modify) Bone - Bone: dyn_hair_braid_5"!(ComponentPose=@n101, Alpha:float=@n2, Translation="0,2,0", Rotation="-2,0,0", Scale="1,1,1") -> (Pose) {
    rotationMode=BMM_Additive
    rotationSpace=BCS_BoneSpace
    translationMode=BMM_Additive
    translationSpace=BCS_BoneSpace
    alphaBoolBlend {blendInTime=0.2, blendOption=Linear, blendOutTime=0.2}
    boneToModify {boneName=dyn_hair_braid_5}
    rotation {pitch=-2}
    translation {y=2}
   }
   n103 = "Transform (Modify) Bone - Bone: dyn_hair_braid_6"(ComponentPose=@n102, Alpha=@n2, Translation="0,-1,0", Rotation="0,-1,0", Scale="1,1,1") {
    rotationMode=BMM_Additive
    rotationSpace=BCS_BoneSpace
    translationMode=BMM_Additive
    translationSpace=BCS_BoneSpace
    alphaBoolBlend {blendInTime=0.2, blendOption=Linear, blendOutTime=0.2}
    boneToModify {boneName=dyn_hair_braid_6}
    rotation {yaw=-1}
    translation {y=-1}
   }
   n104 = "Transform (Modify) Bone - Bone: dyn_hair_braid_6"!(ComponentPose=@n103, bAlphaBoolEnabled:bool=True, Translation="0,2,0", Rotation="0,-5,0", Scale="0.7,0.7,0.7") -> (Pose) {
    alphaInputType=Bool
    scaleMode=BMM_Replace
    scaleSpace=BCS_BoneSpace
    translationMode=BMM_Additive
    translationSpace=BCS_BoneSpace
    alphaBoolBlend {blendOption=Linear}
    boneToModify {boneName=dyn_hair_braid_6}
    rotation {yaw=-5}
    scale=<0.7 0.7 0.7>
    translation {y=2}
   }
   n105 = "Transform (Modify) Bone - Bone: dyn_hair_braid_6"!(ComponentPose=@n104, Alpha:float=@n1, Translation="0,-2,0", Rotation="0,-2,0", Scale="1,1,1") -> (Pose) {
    rotationMode=BMM_Additive
    rotationSpace=BCS_BoneSpace
    translationMode=BMM_Additive
    translationSpace=BCS_BoneSpace
    alphaBoolBlend {blendOption=Linear}
    boneToModify {boneName=dyn_hair_braid_6}
    rotation {yaw=-2}
    translation {y=-2}
   }
   n106 = "Transform (Modify) Bone - Bone: dyn_hair_braid_6"(ComponentPose=@n105, Alpha=Get rb off braid lifting(), Translation="0,-4,0", Rotation="0,-4,0", Scale="1,1,1") {
    rotationMode=BMM_Additive
    rotationSpace=BCS_BoneSpace
    translationMode=BMM_Additive
    translationSpace=BCS_BoneSpace
    alphaBoolBlend {blendOption=Linear}
    boneToModify {boneName=dyn_hair_braid_6}
    rotation {yaw=-4}
    translation {y=-4}
   }
   n107 = Component To Local(ComponentPose=@n106)
   n108 = Get isWearingBackbling()
   n109 = "Transform (Modify) Bone - Bone: dyn_hair_braid_1"(
    ComponentPose=Local To Component(LocalPose=Blend Poses by bool(BlendPose_0=@n107, BlendPose_1=@n81, bActiveValue=@n108) {blendType=Linear})
    Alpha=Get braid_01_jittering()
    Translation="0,-2,0"
    Rotation="0,0,0"
    Scale="1,1,1"
   ) {
    rotationMode=BMM_Additive
    rotationSpace=BCS_BoneSpace
    translationMode=BMM_Additive
    translationSpace=BCS_BoneSpace
    alphaBoolBlend {blendOption=Linear}
    boneToModify {boneName=dyn_hair_braid_1}
    translation {y=-2}
   }
   n110 = Blend Poses by bool(
    BlendPose_0=Component To Local(ComponentPose=@n109)
    BlendPose_1=Blend Poses by bool(
     BlendPose_0=Component To Local(ComponentPose=@n28)
     BlendPose_1=Component To Local(ComponentPose=@n75)
     bActiveValue=Get isWearingBackbling()
    ) {blendType=Linear}
    bActiveValue=Get bRigidBodyEnabled()
   ) {blendType=Linear}
   Output Pose(Result=Component To Local(ComponentPose=Local To Component(LocalPose=@n110))) {layerGroup=None}
   n111 = RigidBody!(
    ComponentPose=Local To Component(LocalPose=Copy Pose From Mesh() {bUseAttachedParent=true})
    Alpha=Get fringe alpha() -> (fringe alpha:double)
    ExternalForce
   ) -> (Pose) {
    bClampLinearTranslationLimitToRefPose=true
    bOverrideWorldGravity=true
    overridePhysicsAsset=&/F_MED_EmberRae_FaceAcc_Physics_fringe
    simulationSpace=BaseBoneSpace
    alphaBoolBlend {blendOption=Linear}
    baseBoneRef {boneName=spine_04}
    componentAppliedLinearAccClamp=<3000 3000 3000>
    componentLinearAccScale=<0.3 0.3 0.3>
    componentLinearVelScale=<0.3 0.3 0.3>
   }
   n112 = Anim Dynamics - Bone: hair_acc_r!(ComponentPose=@n111, Alpha=1, GravityOverride:Vector="0,0,-500", ExternalForce="0,0,0") -> (Pose) {
    angularBiasOverride=0.999
    angularDampingOverride=0.999
    angularSpringConstant=250
    bAngularSpring=true
    bLinearSpring=true
    bOverrideAngularBias=true
    bOverrideAngularDamping=true
    bOverrideLinearDamping=true
    bUseGravityOverride=true
    linearDampingOverride=0.999
    linearSpringConstant=250
    simulationSpace=RootRelative
    alphaBoolBlend {blendOption=Linear}
    alphaScaleBiasClamp {scale=0.4}
    boundBone {boneName=hair_acc_r}
    componentLinearAccScale=<0.2 0.15 0.2>
    componentLinearVelScale=<0.2 0.15 0.2>
    physicsBodyDefinitions [{
      boundBone {boneName=hair_acc_r}
      boxExtents=<5 4 3>
      constraintSetup {angularLimitsMax=<2 2 2>, angularLimitsMin=<-2 -2 -2>, angularTarget=<0 0 0>, linearAxesMax=<2 0.5 2>, linearAxesMin=<-2 -0.5 -2>}
      localJointOffset=<0.4 0.3 0.2>
     }]
   }
   n113 = Anim Dynamics - Bone: hair_acc_l!(ComponentPose=@n112, Alpha=1, GravityOverride:Vector="0,0,-500", ExternalForce="0,0,0") -> (Pose) {
    angularBiasOverride=0.999
    angularDampingOverride=0.999
    angularSpringConstant=250
    bAngularSpring=true
    bLinearSpring=true
    bOverrideAngularBias=true
    bOverrideAngularDamping=true
    bOverrideLinearDamping=true
    bUseGravityOverride=true
    linearDampingOverride=0.999
    linearSpringConstant=250
    simulationSpace=RootRelative
    alphaBoolBlend {blendOption=Linear}
    alphaScaleBiasClamp {scale=0.4}
    boundBone {boneName=hair_acc_l}
    componentLinearAccScale=<0.2 0.15 0.2>
    componentLinearVelScale=<0.2 0.15 0.2>
    physicsBodyDefinitions [{
      boundBone {boneName=hair_acc_l}
      boxExtents=<4 5 4>
      constraintSetup {angularLimitsMax=<2 2 2>, angularLimitsMin=<-2 -2 -2>, angularTarget=<0 0 0>, linearAxesMax=<2 0.5 2>, linearAxesMin=<-2 -0.5 -2>}
      localJointOffset=<0.3 0.4 0.2>
     }]
   }
   Save cached pose 'braid_simulation'(Pose=Component To Local(ComponentPose=@n113)) {cacheName=braid_simulation}
   n114 = Spring controller - Bone: dyn_hair_fr_1_1!(
    ComponentPose:Component Space Pose Link Structure=Local To Component(LocalPose=Copy Pose From Mesh() {bUseAttachedParent=true})
    Alpha:float=1
   ) -> (Pose:Component Space Pose Link Structure) {
    is=title AnimGraphNode_SpringBone
    bRotateX=true
    bRotateY=true
    bRotateZ=true
    bTranslateX=false
    bTranslateY=false
    bTranslateZ=false
    springDamping=8
    springStiffness=200
    alphaBoolBlend {blendOption=Linear}
    alphaScaleBiasClamp {scale=0.1}
    springBone {boneName=dyn_hair_fr_1_1}
   }
   n115 = Spring controller - Bone: dyn_hair_fr_1_2!(ComponentPose:Component Space Pose Link Structure=@n114, Alpha:float=1) -> (Pose:Component Space Pose Link Structure) {
    is=title AnimGraphNode_SpringBone
    bRotateX=true
    bRotateY=true
    bRotateZ=true
    springDamping=8
    springStiffness=390
    alphaBoolBlend {blendOption=Linear}
    alphaScaleBiasClamp {scale=0.1}
    springBone {boneName=dyn_hair_fr_1_2}
   }
   n116 = Spring controller - Bone: dyn_hair_fr_4_1!(ComponentPose:Component Space Pose Link Structure=@n115, Alpha:float=1) -> (Pose:Component Space Pose Link Structure) {
    is=title AnimGraphNode_SpringBone
    bRotateX=true
    bRotateY=true
    bRotateZ=true
    bTranslateX=false
    bTranslateY=false
    bTranslateZ=false
    springDamping=8
    springStiffness=200
    alphaBoolBlend {blendOption=Linear}
    alphaScaleBiasClamp {scale=0.09}
    springBone {boneName=dyn_hair_fr_4_1}
   }
   n117 = Spring controller - Bone: dyn_hair_fr_4_2!(ComponentPose:Component Space Pose Link Structure=@n116, Alpha:float=1) -> (Pose:Component Space Pose Link Structure) {
    is=title AnimGraphNode_SpringBone
    bRotateX=true
    bRotateY=true
    bRotateZ=true
    springDamping=9
    springStiffness=450
    alphaBoolBlend {blendOption=Linear}
    alphaScaleBiasClamp {scale=0.09}
    springBone {boneName=dyn_hair_fr_4_2}
   }
   n118 = Spring controller - Bone: dyn_hair_fr_3_base!(ComponentPose:Component Space Pose Link Structure=@n117, Alpha:float=1) -> (Pose:Component Space Pose Link Structure) {
    is=title AnimGraphNode_SpringBone
    bRotateX=true
    bRotateY=true
    bRotateZ=true
    bTranslateX=false
    bTranslateY=false
    bTranslateZ=false
    springDamping=8
    springStiffness=200
    alphaBoolBlend {blendOption=Linear}
    alphaScaleBiasClamp {scale=0.1}
    springBone {boneName=dyn_hair_fr_3_base}
   }
   n119 = Spring controller - Bone: dyn_hair_fr_3_1!(ComponentPose:Component Space Pose Link Structure=@n118, Alpha:float=1) -> (Pose:Component Space Pose Link Structure) {
    is=title AnimGraphNode_SpringBone
    bRotateX=true
    bRotateY=true
    bRotateZ=true
    springDamping=8
    springStiffness=400
    alphaBoolBlend {blendOption=Linear}
    alphaScaleBiasClamp {scale=0.1}
    springBone {boneName=dyn_hair_fr_3_1}
   }
   n120 = Spring controller - Bone: dyn_hair_fr_3_2!(ComponentPose:Component Space Pose Link Structure=@n119, Alpha:float=1) -> (Pose:Component Space Pose Link Structure) {
    is=title AnimGraphNode_SpringBone
    bRotateX=true
    bRotateY=true
    bRotateZ=true
    springDamping=7
    springStiffness=300
    alphaBoolBlend {blendOption=Linear}
    alphaScaleBiasClamp {scale=0.15}
    springBone {boneName=dyn_hair_fr_3_2}
   }
   n121 = Spring controller - Bone: dyn_hair_fr_2_1!(ComponentPose:Component Space Pose Link Structure=@n120, Alpha:float=1) -> (Pose:Component Space Pose Link Structure) {
    is=title AnimGraphNode_SpringBone
    bRotateX=true
    bRotateY=true
    bRotateZ=true
    bTranslateX=false
    bTranslateY=false
    bTranslateZ=false
    springDamping=8
    springStiffness=200
    alphaBoolBlend {blendOption=Linear}
    alphaScaleBiasClamp {scale=0.09}
    springBone {boneName=dyn_hair_fr_2_1}
   }
   n122 = Spring controller - Bone: dyn_hair_fr_2_2!(ComponentPose:Component Space Pose Link Structure=@n121, Alpha:float=1) -> (Pose:Component Space Pose Link Structure) {
    is=title AnimGraphNode_SpringBone
    bRotateX=true
    bRotateY=true
    bRotateZ=true
    springDamping=6
    springStiffness=390
    alphaBoolBlend {blendOption=Linear}
    alphaScaleBiasClamp {scale=0.1}
    springBone {boneName=dyn_hair_fr_2_2}
   }
   n123 = Spring controller - Bone: dyn_hair_fr_5_1!(ComponentPose:Component Space Pose Link Structure=@n122, Alpha:float=1) -> (Pose:Component Space Pose Link Structure) {
    is=title AnimGraphNode_SpringBone
    bRotateX=true
    bRotateY=true
    bRotateZ=true
    bTranslateX=false
    bTranslateY=false
    bTranslateZ=false
    springDamping=8
    springStiffness=200
    alphaBoolBlend {blendOption=Linear}
    alphaScaleBiasClamp {scale=0.15}
    springBone {boneName=dyn_hair_fr_5_1}
   }
   n124 = Spring controller - Bone: dyn_hair_fr_5_2!(ComponentPose:Component Space Pose Link Structure=@n123, Alpha:float=1) -> (Pose:Component Space Pose Link Structure) {
    is=title AnimGraphNode_SpringBone
    bRotateX=true
    bRotateY=true
    bRotateZ=true
    springDamping=7
    springStiffness=430
    alphaBoolBlend {blendOption=Linear}
    alphaScaleBiasClamp {scale=0.15}
    springBone {boneName=dyn_hair_fr_5_2}
   }
   n125 = Anim Dynamics - Bone: hair_acc_r(ComponentPose=@n124, Alpha=1, GravityOverride="0,0,-500", ExternalForce="0,0,0") {
    angularBiasOverride=0.999
    angularDampingOverride=0.999
    angularSpringConstant=250
    bAngularSpring=true
    bLinearSpring=true
    bOverrideAngularBias=true
    bOverrideAngularDamping=true
    bOverrideLinearDamping=true
    bUseGravityOverride=true
    linearDampingOverride=0.999
    linearSpringConstant=250
    simulationSpace=RootRelative
    alphaBoolBlend {blendOption=Linear}
    alphaScaleBiasClamp {scale=0.4}
    boundBone {boneName=hair_acc_r}
    componentLinearAccScale=<0.2 0.15 0.2>
    componentLinearVelScale=<0.2 0.15 0.2>
    physicsBodyDefinitions [{
      boundBone {boneName=hair_acc_r}
      boxExtents=<5 4 3>
      constraintSetup {angularLimitsMax=<2 2 2>, angularLimitsMin=<-2 -2 -2>, angularTarget=<0 0 0>, linearAxesMax=<2 0.5 2>, linearAxesMin=<-2 -0.5 -2>}
      localJointOffset=<0.4 0.3 0.2>
     }]
   }
   n126 = Anim Dynamics - Bone: hair_acc_l(ComponentPose=@n125, Alpha=1, GravityOverride="0,0,-500", ExternalForce="0,0,0") {
    angularBiasOverride=0.999
    angularDampingOverride=0.999
    angularSpringConstant=250
    bAngularSpring=true
    bLinearSpring=true
    bOverrideAngularBias=true
    bOverrideAngularDamping=true
    bOverrideLinearDamping=true
    bUseGravityOverride=true
    linearDampingOverride=0.999
    linearSpringConstant=250
    simulationSpace=RootRelative
    alphaBoolBlend {blendOption=Linear}
    alphaScaleBiasClamp {scale=0.4}
    boundBone {boneName=hair_acc_l}
    componentLinearAccScale=<0.2 0.15 0.2>
    componentLinearVelScale=<0.2 0.15 0.2>
    physicsBodyDefinitions [{
      boundBone {boneName=hair_acc_l}
      boxExtents=<4 5 4>
      constraintSetup {angularLimitsMax=<2 2 2>, angularLimitsMin=<-2 -2 -2>, angularTarget=<0 0 0>, linearAxesMax=<2 0.5 2>, linearAxesMin=<-2 -0.5 -2>}
      localJointOffset=<0.3 0.4 0.2>
     }]
   }
   n127 = Get alpha jump() -> (alpha jump:double)
   n128 = "Transform (Modify) Bone - Bone: dyn_hair_fr_3_2"(ComponentPose=@n126, Alpha=@n127, Translation="0,0,0", Rotation="-100,0,0", Scale="1,1,1") {
    rotationMode=BMM_Additive
    rotationSpace=BCS_BoneSpace
    alphaBoolBlend {blendOption=Linear}
    boneToModify {boneName=dyn_hair_fr_3_2}
    rotation {pitch=-100}
   }
   n129 = "Transform (Modify) Bone - Bone: dyn_hair_fr_3_base"(ComponentPose=@n128, Alpha=@n127, Translation="0,0,0", Rotation="-60,0,0", Scale="1,1,1") {
    rotationMode=BMM_Additive
    rotationSpace=BCS_BoneSpace
    alphaBoolBlend {blendOption=Linear}
    boneToModify {boneName=dyn_hair_fr_3_base}
    rotation {pitch=-60}
   }
   n130 = "Transform (Modify) Bone - Bone: dyn_hair_braid_1"(ComponentPose=@n129, Alpha=0, Translation="0,0,0", Rotation="-0.023,0,-0.028", Scale="1,1,1") {
    rotationMode=BMM_Additive
    rotationSpace=BCS_BoneSpace
    alphaBoolBlend {blendOption=Linear}
    boneToModify {boneName=dyn_hair_braid_1}
    rotation {pitch=-0.023, roll=-0.028}
   }
   Save cached pose 'low_perf_fringe'(Pose=Component To Local(ComponentPose=@n130)) {cacheName=low_perf_fringe}
  ]}
 EventGraph {nodes [
   n1 = Event BlueprintUpdateAnimation()
   > n2 = Set deltaTime(deltaTime:double=@n1.DeltaTimeX) -> (Output_Get:double)
   > Set relaxation fps(relaxation fps:double=MapRangeClamped(Value=float / float(A=1, B=@n2), InRangeA=25, InRangeB=60, OutRangeA=1.05, OutRangeB=2)) -> (Output_Get:double)
   > n3 = Sequence!(execute) -> (
    then_0
    then_1
    then_2:Exec
    then_3:Exec
    then_4:Exec
    then_5:Exec
    then_6:Exec
    then_7:Exec
    then_8:Exec
    then_9:Exec
    then_10:Exec
    then_11:Exec
    then_12:Exec
    then_13:Exec
    then_14:Exec
    then_15:Exec
   )
   > n4 = Gravity Override State Function DEPRECATED(CharacterPartAnimInstance=Self-Reference(), GameTimeInSeconds=GetGameTimeInSeconds(), GravityOverrideParams=Get Gravity Override Params())
   > Set Gravity Override(Gravity Override=@n4)
   n5 = Branch(execute=@n3.then_10, Condition=Get bIsJumping())
   > Set sky diving braid 6(sky diving braid 6:double=FInterpTo(Current=Get sky diving braid 6() -> (sky diving braid 6:double), Target=0.8, DeltaTime=Get deltaTime() -> (deltaTime:double), InterpSpeed=1)) -> (Output_Get:double)
   Set sky diving braid 6(execute=@n5.else, sky diving braid 6=FInterpTo(Current=Get sky diving braid 6(), Target=0, DeltaTime=Get deltaTime(), InterpSpeed=1))
   n6 = NOT Boolean(A=Get isPickaxe() -> (isPickaxe:bool))
   Set brai 01 jittering GUN(
    execute=@n3.then_5
    brai 01 jittering GUN:double=float * float(A=Get braid_01_jittering(), B="To Float (Boolean)"(InBool=@n6)) {is=title K2Node_PromotableOperator}
   ) -> (Output_Get:double)
   n7 = AND Boolean(A=Get bIsCrouching(), B=false)
   > Set braid jittering CROUCH(braid jittering CROUCH:double=float * float(A=Get braid_01_jittering(), B="To Float (Boolean)"(InBool=@n7))) -> (Output_Get:double)
   n8 = Branch(execute=@n3.then_6, Condition=OR Boolean(A=Get bIsJumping(), B=Get bIsFalling()))
   > Set jumping_cape_bone(jumping_cape_bone:bool=true) -> (Output_Get:bool)
   Set jumping_cape_bone(execute=@n8.else, jumping_cape_bone=false)
   n9 = Branch(execute=@n3.then_8, Condition=OR Boolean(A=Get bIsFalling(), B=Get bIsJumping()))
   > n10 = Set ext force braid(ext force braid:double=4500) -> (Output_Get:double)
   n11 = "To Float (Boolean)"(InBool=Get bIsMoving2D())
   n12 = Vector Length(A=Get PawnVelocity())
   Set fringe alpha(
    execute=[@n13.then, @n10.then]
    fringe alpha:double=MapRangeClamped(Value=float * float(A=@n12, B=@n11), InRangeA=20, InRangeB=770, OutRangeA=1, OutRangeB=0.2)
   ) -> (Output_Get:double)
   n13 = Set ext force braid(execute=@n9.else, ext force braid=0)
   n14 = Branch(execute=@n3.then_9, Condition=Get bIsPlayingMeleeAnim())
   > Set playing melee braid 6(playing melee braid 6:double=FInterpTo(Current=Get playing melee braid 6() -> (playing melee braid 6:double), Target=0.6, DeltaTime=Get deltaTime(), InterpSpeed=5)) -> (Output_Get:double)
   Set playing melee braid 6(execute=@n14.else, playing melee braid 6=FInterpTo(Current=Get playing melee braid 6(), Target=0, DeltaTime=Get deltaTime(), InterpSpeed=5))
   Set rb off braid lifting(
    execute=@n3.then_11
    rb off braid lifting:double=GetAlphaFromReferenceTransformDelta(
     SkeletalMeshComponent=Get Skeletal Mesh Component() -> (Skeletal Mesh Component:Skeletal Mesh Component Object Reference)
     BoneName=head
     BaseName=spine_05
     InputRange=(X=0,Y=20)
     TwistAxis="1,0,0"
    )
   ) -> (Output_Get:double)
   > Set rb off lifting standing(rb off lifting standing:double=GetAlphaFromReferenceTransformDelta(SkeletalMeshComponent=Get Skeletal Mesh Component(), BoneName=head, BaseName=spine_05, InputRange=(X=-13,Y=-46), TwistAxis="1,0,0")) -> (Output_Get:double)
   > Set shooting alpha rb off standing(shooting alpha rb off standing:float=GetAlphaFromReferenceTransformDelta(
     SkeletalMeshComponent=Get Skeletal Mesh Component()
     BoneName=upperarm_r
     BaseName=spine_05
     InputRange=(X=70,Y=90)
     OutputRange=(X=0,Y=0.4)
     TwistAxis="1,0,0"
    )) -> (Output_Get:float)
   Set braid_01_jittering(
    execute=@n3.then_4
    braid_01_jittering:double=GetAlphaFromReferenceTransformDelta(SkeletalMeshComponent=Get Skeletal Mesh Component(), BoneName=head, BaseName=spine_05, InputRange=(X=-1,Y=-52), TwistAxis="1,0,0")
   ) -> (Output_Get:double)
   > n15 = Switch on EFortWeaponCoreAnimation!(execute, Selection:EFortWeaponCoreAnimation Enum=Get WeaponCoreAnim!() -> (WeaponCoreAnim)) -> (
    Melee:Exec
    Pistol:Exec
    Shotgun:Exec
    PaperBlueprint:Exec
    Rifle:Exec
    MeleeOneHand:Exec
    MachinePistol:Exec
    RocketLauncher:Exec
    GrenadeLauncher:Exec
    GoingCommando:Exec
    AssaultRifle:Exec
    TacticalShotgun:Exec
    SniperRifle:Exec
    TrapPlacement:Exec
    ShoulderLauncher:Exec
    AbilityDecoTool:Exec
    Crossbow:Exec
    C4:Exec
    RemoteControl:Exec
    DualWield:Exec
    AR_BullPup:Exec
    AR_ForwardGrip:Exec
    MedPackPaddles:Exec
    SMG_P90:Exec
    AR_DrumGun:Exec
    Consumable_Small:Exec
    Consumable_Large:Exec
    Balloon:Exec
    MountedTurret:Exec
    CreativeTool:Exec
    ExplosiveBow:Exec
    AshtonIndigo:Exec
    AshtonChicago:Exec
    MeleeDualWield:Exec
    Unarmed:Exec
   )
   Set isPickaxe(
    execute=[
     @n15.AR_BullPup
     @n15.AR_DrumGun
     @n15.AR_ForwardGrip
     @n15.AbilityDecoTool
     @n15.AshtonChicago
     @n15.AshtonIndigo
     @n15.AssaultRifle
     @n15.Balloon
     @n15.C4
     @n15.Consumable_Large
     @n15.Consumable_Small
     @n15.CreativeTool
     @n15.Crossbow
     @n15.DualWield
     @n15.ExplosiveBow
     @n15.GoingCommando
     @n15.GrenadeLauncher
     @n15.MachinePistol
     @n15.MedPackPaddles
     @n15.MountedTurret
     @n15.PaperBlueprint
     @n15.Pistol
     @n15.Rifle
     @n15.RocketLauncher
     @n15.SMG_P90
     @n15.Shotgun
     @n15.ShoulderLauncher
     @n15.SniperRifle
     @n15.TacticalShotgun
     @n15.TrapPlacement
    ]
    isPickaxe:bool=false
   ) -> (Output_Get:bool)
   Set isPickaxe(execute=[@n15.Melee, @n15.MeleeDualWield, @n15.MeleeOneHand], isPickaxe=true)
   Set arms up(
    execute=@n3.then_12
    arms up:float=GetAlphaFromReferenceTransformDelta(
     SkeletalMeshComponent=Get Skeletal Mesh Component()
     BoneName=upperarm_r
     BaseName=spine_05
     InputRange=(X=60,Y=120)
     OutputRange=(X=0.1,Y=0)
     TwistAxis="1,0,0"
    )
   ) -> (Output_Get:float)
   > Set arms up2(arms up2:double=GetAlphaFromReferenceTransformDelta(SkeletalMeshComponent=Get Skeletal Mesh Component(), BoneName=upperarm_r, BaseName=spine_05, InputRange=(X=60,Y=120), TwistAxis="1,0,0")) -> (Output_Get:double)
   > Sequence!(execute) -> (then_0, then_1)
   > n16 = Branch(Condition=NOT Boolean(A=Get isPickaxe()))
   > Set rb alpha(rb alpha:double=0.06) -> (Output_Get:double)
   Set rb alpha(execute=@n16.else, rb alpha=0.15)
   n17 = Branch(execute=@n3.then_2, Condition=Get bIsCrouching())
   > Set crouching braid rb off(crouching braid rb off:double=FInterpTo(Current=Get crouching braid rb off() -> (crouching braid rb off:double), Target=1, DeltaTime=Get deltaTime(), InterpSpeed=5)) -> (Output_Get:double)
   > n18 = Branch(Condition=Get bIsMoving2D())
   > Set crouching and moving 2d(crouching and moving 2d:double=FInterpTo(Current=Get crouching and moving 2d(), Target=1, DeltaTime=Get deltaTime(), InterpSpeed=10)) -> (Output_Get:double)
   Set crouching and moving 2d(
    execute=@n18.else
    crouching and moving 2d=FInterpTo(Current=Get crouching and moving 2d(), Target=0, DeltaTime=Get deltaTime(), InterpSpeed=10)
   )
   Set crouching braid rb off(
    execute=@n17.else
    crouching braid rb off=FInterpTo(Current=Get crouching braid rb off(), Target=1, DeltaTime=Get deltaTime(), InterpSpeed=5)
   )
   > Branch(Condition=Get bIsMoving2D())
   > Set crouching braid rb off(crouching braid rb off=FInterpTo(Current=Get crouching braid rb off(), Target=1.5, DeltaTime=Get deltaTime(), InterpSpeed=2))
   n19 = Branch(execute=@n3.then_13, Condition=Get bIsSkydiveDiveMode())
   > Set sky diving clipping alpha(sky diving clipping alpha:double=FInterpTo(Current=Get sky diving clipping alpha(), Target=0.4, DeltaTime=Get deltaTime(), InterpSpeed=10)) -> (Output_Get:double)
   Set sky diving clipping alpha(
    execute=@n19.else
    sky diving clipping alpha=FInterpTo(Current=Get sky diving clipping alpha(), Target=0, DeltaTime=Get deltaTime(), InterpSpeed=1)
   )
   n20 = Get bIsCrouching()
   Set braid trail alpha(
    execute=@n3.then_14
    braid trail alpha:double=FInterpTo(Current=Get braid trail alpha(), Target=SelectFloat(A=0.25, B=0, bPickA=@n20), DeltaTime=Get deltaTime(), InterpSpeed=10)
   ) -> (Output_Get:double)
   n21 = Get bIsPlayingMeleeAnim()
   > Set braid trail alpha(braid trail alpha=FInterpTo(Current=Get braid trail alpha(), Target=SelectFloat(A=0.5, B=0, bPickA=@n21), DeltaTime=Get deltaTime(), InterpSpeed=10))
   > Set braid clipping crouching trail(braid clipping crouching trail:float=GetAlphaFromReferenceTransformDelta(SkeletalMeshComponent=Get Skeletal Mesh Component(), BoneName=upperarm_l, BaseName=spine_05, InputRange=(X=50,Y=98), TwistAxis="1,0,0")) -> (Output_Get:float)
   > Set sliding trail(sliding trail:double=GetAlphaFromReferenceTransformDelta(
     SkeletalMeshComponent=Get Skeletal Mesh Component()
     BoneName=spine_05
     BaseName=spine_01
     InputRange=(X=50,Y=37)
     OutputRange=(X=-8,Y=0)
     TwistAxis="1,0,0"
    )) -> (Output_Get:double)
   n22 = AND Boolean(A=AND Boolean(A=Get bIsCrouching(), B=Get isPickaxe()), B=NOT Boolean(A=Get bIsPlayingMeleeAnim()))
   > Set crouching pickaxe(crouching pickaxe:double=FInterpTo(
     Current=Get crouching pickaxe() -> (crouching pickaxe:double)
     Target=SelectFloat(A=0.9, B=0, bPickA=@n22)
     DeltaTime=Get deltaTime()
     InterpSpeed=1
    )) -> (Output_Get:double)
   > Set melee and crouching(melee and crouching:bool=AND Boolean(A=Get bIsPlayingMeleeAnim(), B=Get bIsCrouching())) -> (Output_Get:bool)
   n23 = Get bIsCrouching()
   > Set braid trail alpha 2(braid trail alpha 2:double=FInterpTo(
     Current=Get braid trail alpha 2() -> (braid trail alpha 2:double)
     Target=SelectFloat(A=0.6, B=0, bPickA=@n23)
     DeltaTime=Get deltaTime()
     InterpSpeed=8
    )) -> (Output_Get:double)
   n24 = Get playing melee braid 6()
   n25 = Get braid trail alpha 2()
   > Set melee or crouching dynamics(melee or crouching dynamics:double="Clamp (Float)"(Value=float + float(A=@n24, B=@n25), Min=0, Max=0.6)) -> (Output_Get:double)
   n26 = Get bIsMoving2D()
   > Set trail alpha moving 2d(trail alpha moving 2d:double=FInterpTo(Current=Get braid trail alpha(), Target=SelectFloat(A=0.7, B=0, bPickA=@n26), DeltaTime=Get deltaTime(), InterpSpeed=10)) -> (Output_Get:double)
   n27 = GetAlphaFromReferenceTransformDelta(
    SkeletalMeshComponent=Get Skeletal Mesh Component()
    BoneName=spine_05
    BaseName=spine_01
    DeltaComponentType=EulerX
    InputRange=(X=-25,Y=-38)
    OutputRange=(X=0,Y=0.7)
    TwistAxis="1,0,0"
   )
   n28 = GetAlphaFromReferenceTransformDelta(
    SkeletalMeshComponent=Get Skeletal Mesh Component()
    BoneName=spine_05
    BaseName=spine_01
    DeltaComponentType=EulerX
    InputRange=(X=-5,Y=13)
    OutputRange=(X=0,Y=0.7)
    TwistAxis="1,0,0"
   )
   > Set trail rotating alpha(trail rotating alpha:double="Clamp (Float)"(Value=float + float(A=@n27, B=@n28), Min=0, Max=1)) -> (Output_Get:double)
   n29 = AND Boolean(A=Get bIsMoving2D(), B=Get isPickaxe())
   > Set trail alpha moving 2d pickaxe(trail alpha moving 2d pickaxe:double=FInterpTo(
     Current=Get trail alpha moving 2d pickaxe() -> (trail alpha moving 2d pickaxe:double)
     Target=SelectFloat(A=0.2, B=0, bPickA=@n29)
     DeltaTime=Get deltaTime()
     InterpSpeed=10
    )) -> (Output_Get:double)
   n30 = "To Float (Boolean)"(InBool=NOT Boolean(A=Get bIsMoving2D()))
   n31 = float + float!(
    A=Get trail rotating alpha() -> (trail rotating alpha:double)
    B=Get rb off braid lifting()
    C:float=GetAlphaFromReferenceTransformDelta(
     SkeletalMeshComponent=Get Skeletal Mesh Component()
     BoneName=spine_05
     BaseName=spine_01
     InputRange=(X=2,Y=15)
     OutputRange=(X=0,Y=0.5)
     TwistAxis="1,0,0"
    )
   ) -> (ReturnValue)
   n32 = "Clamp (Float)"(
    Value=float + float!(A, B=Get braid_01_jittering(), C:double=Get trail alpha moving 2d pickaxe(), D:double=float * float(A=@n30, B=@n31), E:double) -> (ReturnValue)
    Min=0
    Max=1
   )
   n33 = Get bIsJumping()
   n34 = Get bIsFalling()
   n35 = "To Float (Boolean)"(InBool=NOT Boolean(A=OR Boolean(A=@n33, B=@n34)))
   > Set trail alpha moving 2d 2(trail alpha moving 2d 2:double=FInterpTo(
     Current=Get trail alpha moving 2d 2() -> (trail alpha moving 2d 2:double)
     Target=float * float(A=@n32, B=@n35)
     DeltaTime=Get deltaTime()
     InterpSpeed=7
    )) -> (Output_Get:double)
   n36 = Get bIsPlayingMeleeAnim()
   > Set crouching not melee(crouching not melee:bool=AND Boolean(A=Get bIsCrouching(), B=NOT Boolean(A=@n36))) -> (Output_Get:bool)
   > Set alpha crouching or jumping(alpha crouching or jumping:bool=OR Boolean(A=Get crouching_alpha_braid() -> (crouching_alpha_braid:bool), B=Get bIsJumping())) -> (Output_Get:bool)
   n37 = OR Boolean(A=Get bIsJumping(), B=Get bIsFalling())
   Set isjumping(
    execute=@n3.then_1
    isjumping:double=FInterpTo(Current=Get isjumping() -> (isjumping:double), Target=SelectFloat(A=0.09, B=0.5, bPickA=@n37), DeltaTime=Get deltaTime(), InterpSpeed=2)
   ) -> (Output_Get:double)
   n38 = Branch(execute=@n3.then_3, Condition=Get bIsJumping())
   > Set alpha jump(alpha jump:double=FInterpTo(Current=Get alpha jump(), Target=1, DeltaTime=Get deltaTime(), InterpSpeed=0.5)) -> (Output_Get:double)
   Set alpha jump(execute=@n38.else, alpha jump=FInterpTo(Current=Get alpha jump(), Target=0, DeltaTime=Get deltaTime(), InterpSpeed=1))
   Set cape braid rotation(
    execute=@n3.then_7
    cape braid rotation:double=GetAlphaFromReferenceTransformDelta(
     SkeletalMeshComponent=Get Skeletal Mesh Component()
     BoneName=head
     BaseName=spine_05
     DeltaComponentType=EulerX
     InputRange=(X=-2,Y=-15)
     TwistAxis="1,0,0"
    )
   ) -> (Output_Get:double)
   > Set cape end braid rotation jump(cape end braid rotation jump:bool=OR Boolean(A=Get bIsJumping(), B=Get bIsFalling())) -> (Output_Get:bool)
   Event OnCharacterCustomizationLoaded()
   > n39 = Sequence!(execute) -> (then_0, then_1, then_2:Exec, then_3:Exec)
   n40 = CheckForBackbling()
   > Set isWearingBackbling(isWearingBackbling=OR Boolean(A=@n40, B=@n40.bIsBackblingACape))
   > Set isWearingBackpack(isWearingBackpack:bool=@n40) -> (Output_Get:bool)
   > Set isWearingCapeBool(isWearingCapeBool:bool=@n40.bIsBackblingACape) -> (Output_Get:bool)
   > n41 = Branch(Condition=Get isWearingBackbling())
   > Set backpack_alpha(backpack_alpha:double=1) -> (Output_Get:double)
   Set backpack_alpha(execute=@n41.else, backpack_alpha=0)
   n42 = Cast To FortPlayerPawn(execute=@n39.then_1, Object=TryGetPawnOwner())
   n43 = GetSkeletalMeshForPartType(self=@n42, PartType=Backpack)
   > n44 = Branch(Condition=Contains(SearchIn=GetDisplayName(Object=@n43), Substring=Hightower))
   > Set is bad cape(is bad cape:double=0) -> (Output_Get:double)
   > Set is bad cape transformation(is bad cape transformation:bool=true) -> (Output_Get:bool)
   Set is bad cape(execute=@n44.else, is bad cape=0.15)
   > Set is bad cape transformation(is bad cape transformation=false)
   Set Skeletal Mesh Component(execute=@n39.then_2, Skeletal Mesh Component:Skeletal Mesh Component Object Reference=GetOwnerSkeletalMeshComponent()) -> (Output_Get:Skeletal Mesh Component Object Reference)
   n45 = Event OnBeginCrouching()
   Do Once(execute=[@n45.then, @n47.then], Reset=@n46.then)
   > Set crouching_alpha_braid(crouching_alpha_braid:bool=true) -> (Output_Get:bool)
   > Delay(Duration=0.2)
   > n46 = Set crouching_alpha_braid(crouching_alpha_braid=false)
   n47 = Event OnEndCrouching()
   Event OnWeaponChanged()
   > Branch(Condition=Get bIsCrouching())
   > Set weapon changed(weapon changed:double=1) -> (Output_Get:double)
   > Delay(Duration=0.1)
   > Set weapon changed(weapon changed=0)
   Contains(Substring=GoingUp)
   NOT Boolean(A=Get isPickaxe())
   Set gravity fix(gravity fix:bool=false) -> (Output_Get:bool)
  ]}
]
variables [
 alpha crouching or jumping {type=boolean}
 alpha jump {type=number}
 arms up {type=number}
 arms up2 {type=number}
 backpack_alpha {type=number}
 brai 01 jittering GUN {type=number}
 braid clipping crouching trail {type=number}
 braid jittering CROUCH {type=number}
 braid trail alpha {type=number}
 braid trail alpha 2 {type=number}
 braid_01_jittering {type=number}
 cape braid rotation {type=number}
 cape end braid rotation jump {type=boolean}
 crouching and moving 2d {type=number}
 crouching braid rb off {type=number}
 crouching not melee {type=boolean}
 crouching pickaxe {type=number}
 crouching_alpha_braid {type=boolean}
 deltaTime {type=number}
 ext force braid {type=number}
 fringe alpha {type=number}
 gravity fix {type=boolean}
 Gravity Override {type=Vector}
 Gravity Override Params {type=GravityOverrideParamsStruct, default {skydive_FloatingDown {gravityOverride=<0 -700 1500>}, skydive_GoingUp {gravityOverride=<0 0 0>}}}
 is bad cape {type=number}
 is bad cape transformation {type=boolean}
 isjumping {type=number, default=2.1429}
 isPickaxe {type=boolean}
 isWearingBackbling {type=boolean}
 isWearingBackpack {type=boolean}
 isWearingCapeBool {type=boolean}
 jumping_cape_bone {type=boolean}
 melee and crouching {type=boolean}
 melee or crouching dynamics {type=number}
 NewVar {type=boolean}
 playing melee braid 6 {type=number}
 rb alpha {type=number}
 rb off braid lifting {type=number}
 rb off lifting standing {type=number}
 relaxation fps {type=number}
 shooting alpha rb off standing {type=number}
 Skeletal Mesh Component {type=SkeletalMeshComponent}
 sky diving alpha {type=number}
 sky diving braid 6 {type=number}
 sky diving clipping alpha {type=number}
 sliding trail {type=number}
 trail alpha moving 2d {type=number}
 trail alpha moving 2d 2 {type=number}
 trail alpha moving 2d pickaxe {type=number}
 trail rotating alpha {type=number}
 weapon changed {type=number}
]
